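# Physics asset for the dangling chain on the F_MED_SilentTempo female character body.
# Two capsule bodies form the chain links, constrained in series to a small master chain body.
# Angular limits favor one swing axis, wider at the lower joint, with tight side swing and twist for a nearly planar sway.
# An unconstrained left thigh capsule appears to serve as a collider for the chain against the leg.
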PhysicsAsset /BRCosmetics/Characters/Player/Female/Medium/Bodies/F_MED_SilentTempo/Meshes/F_MED_SilentTempo_Chain_Physics
bodies {
 dyn_chain_sim_01 [{shapeType=Capsule, length=16.814, radius=1.4654, shapeName=dyn_chain_sim_01_capsule, center=<10.0185 0 0>}]
 dyn_chain_sim_02 [{shapeType=Capsule, length=15.9869, radius=1.7426, shapeName=dyn_chain_sim_02_capsule, center=<10.0185 0 0>}]
 dyn_mstr_chain_sim [{shapeType=Capsule, length=0, radius=0.5, shapeName=dyn_mstr_chain_sim_capsule}]
 thigh_l [{
   shapeType=Capsule
   length=34.4627
   radius=7.0373
   shapeName=thigh_l_capsule
   center=<-18.3898 -0.3771 0.0013>
   rotation=<rot -89.5392 0.8943 -0.8739>
  }]
}
constraints [
 {bone1Name=dyn_chain_sim_02, bone2Name=dyn_chain_sim_01, swing1LimitDegrees=35, swing2LimitDegrees=5, twistLimitDegrees=2}
 {bone1Name=dyn_chain_sim_01, bone2Name=dyn_mstr_chain_sim, swing1LimitDegrees=25, swing2LimitDegrees=5, twistLimitDegrees=2}
]
