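# Control Rig for the F_MED_SilentTempo character body that turns the hour, minute and second hands of the hanging clock.
# It scales game_time_in_seconds by four, then derives each hand's angle with modulo math.
# A sequence applies the negated angles as Y rotations to the three dfrm_clock bones.
# The hour angle adds a minute offset to already fractional hours, so it double-counts minutes and jumps back each hour.
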ControlRigBlueprint /BRCosmetics/Characters/Player/Female/Medium/Bodies/F_MED_SilentTempo/Meshes/F_MED_SilentTempo_Clock_CtrlRig
graphs [
 RigVMFunctionLibrary {links [], nodes []}
 RigVMModel {nodes [
   RigUnit_BeginExecution = &RigUnit_BeginExecution()
   SequenceAggregate = RigVMAggregateNode(ExecuteContext=@RigUnit_BeginExecution.ExecutePin)
   n1 = &RigVMFunction_MathDoubleMul(B=4, A=RigVMVariableNode() {variable=game_time_in_seconds}.Value)
   Modulo = /Script/RigVM.RigVMFunction_MathDoubleMod(B=60, A=@n1.Result) {is=struct RigVMUnitNode, template="Modulo::Execute(in A,in B,out Result)"}
   Multiply = &RigVMFunction_MathDoubleMul(B=-1, A=&RigVMFunction_MathDoubleMul(B=6, A=@Modulo.Result).Result)
   &RigUnit_ModifyTransforms(
    ItemToModify=((Item=(Type=Bone,Name="dfrm_clock_second"),Transform=(Rotation=(X=0,Y=0,Z=0,W=1),Translation=(X=0,Y=0,Z=0),Scale3D=(X=1,Y=1,Z=1))))
    Weight=1
    ItemToModify.0.Transform.Rotation=&RigVMFunction_MathQuaternionFromEuler(Euler=(X=0,Y=0,Z=0), Euler.Y=@Multiply.Result).Result
    ExecutePin=@SequenceAggregate.A
   )
   Divide = &RigVMFunction_MathDoubleDiv(B=60, A=&RigVMFunction_MathDoubleMod(B=3600, A=@n1.Result) {template="Modulo::Execute(in A,in B,out Result)"}.Result)
   n2 = &RigVMFunction_MathDoubleMul(B=6, A=@Divide.Result)
   &RigUnit_ModifyTransforms(
    ItemToModify=((Item=(Type=Bone,Name="dfrm_clock_minute"),Transform=(Rotation=(X=0,Y=0,Z=0,W=1),Translation=(X=0,Y=0,Z=0),Scale3D=(X=1,Y=1,Z=1))))
    Weight=1
    ItemToModify.0.Transform.Rotation=&RigVMFunction_MathQuaternionFromEuler(Euler=(X=0,Y=0,Z=0), Euler.Y=&RigVMFunction_MathDoubleMul(B=-1, A=@n2.Result).Result).Result
    ExecutePin=@SequenceAggregate.B
   )
   n3 = &RigVMFunction_MathDoubleMod(B=43200, A=@n1.Result) {template="Modulo::Execute(in A,in B,out Result)"}
   n4 = &RigVMFunction_MathDoubleMod(B=3600, A=@n1.Result) {template="Modulo::Execute(in A,in B,out Result)"}
   Add = &RigVMFunction_MathDoubleAdd(
    A=&RigVMFunction_MathDoubleMul(B=30, A=&RigVMFunction_MathDoubleDiv(B=3600, A=@n3.Result).Result).Result
    B=&RigVMFunction_MathDoubleMul(B=0.5, A=&RigVMFunction_MathDoubleDiv(B=60, A=@n4.Result).Result).Result
   )
   &RigUnit_ModifyTransforms(
    ItemToModify=((Item=(Type=Bone,Name="dfrm_clock_hour"),Transform=(Rotation=(X=0,Y=0,Z=0,W=1),Translation=(X=0,Y=0,Z=0),Scale3D=(X=1,Y=1,Z=1))))
    Weight=1
    ItemToModify.0.Transform.Rotation=&RigVMFunction_MathQuaternionFromEuler(Euler=(X=0,Y=0,Z=0), Euler.Y=&RigVMFunction_MathDoubleMul(B=-1, A=@Add.Result).Result).Result
    ExecutePin=@SequenceAggregate.C
   )
  ]}
 SequenceAggregate_1_ContainedGraph {nodes [
   n1 = RigVMFunctionEntryNode()
   SequenceAggregate = &RigVMFunction_Sequence(ExecuteContext=@n1.ExecuteContext)
   SequenceAggregate_1 = &RigVMFunction_Sequence(ExecuteContext=@SequenceAggregate.B)
   RigVMFunctionReturnNode(A=@SequenceAggregate.A, B=@SequenceAggregate_1.A, C=@SequenceAggregate_1.B)
  ]}
]
hierarchy [
 root {loc=<0 0 0>, rot=<0 0 0 1>}
 pelvis {parent=root, loc=<0 -0.5301 98.0571>, rot=<0.0127 0.707 -0.0127 -0.707>}
 spine_01 {parent=pelvis, loc=<6.1208 -0.9328 0>, rot=<0 0 0.034 -0.9994>}
 spine_02 {parent=spine_01, loc=<6.5996 -0.2095 0>, rot=<0 0 0.0238 -0.9997>}
 spine_03 {parent=spine_02, loc=<6.4914 -0.1237 0>, rot=<0 0 0.0149 -0.9999>}
 spine_04 {parent=spine_03, loc=<6.9598 0.0794 0>, rot=<0 0 -0.0132 -0.9999>}
 spine_05 {parent=spine_04, loc=<6.9987 0 0>, rot=<0 0 -0.0015 -1>}
 attach_cape {parent=spine_05, loc=<9.3216 -8.4501 0>, rot=<0.0616 -0.7044 -0.0616 -0.7044>}
 neck_01 {parent=spine_05, loc=<11.3138 0.1529 0>, rot=<0 0 -0.2003 -0.9797>}
 neck_02 {parent=neck_01, loc=<4.5627 0 0>, rot=<0 0 0.0504 -0.9987>}
 head {parent=neck_02, loc=<5.3914 0 0>, rot=<0 0 0.115 -0.9934>}
 dyn_hair_offset {parent=head, loc=<4 0 0>, rot=<0 0 0 1>}
 dyn_hair_a_r_base {parent=dyn_hair_offset, loc=<1.3722 4.4493 7.2537>, rot=<0.5932 -0.6309 0.3671 0.3395>}
 dyn_hair_a_r_parent {parent=dyn_hair_a_r_base, loc=<0 0 0>, rot=<0 0 0 1>}
 dyn_hair_c_r_base {parent=dyn_hair_offset, loc=<2.8768 -0.2035 8.0968>, rot=<-0.2466 -0.9123 0.065 0.3202>}
 dyn_hair_c_r_parent {parent=dyn_hair_c_r_base, loc=<0 0 0>, rot=<0 0 0 1>}
 dyn_hair_d_r_base {parent=dyn_hair_offset, loc=<-1.5019 -1.9539 8.3757>, rot=<-0.2422 -0.9356 -0.0972 0.238>}
 dyn_hair_d_r_parent {parent=dyn_hair_d_r_base, loc=<0 0 0>, rot=<0 0 0 1>}
 dyn_hair_e_l_base {parent=dyn_hair_offset, loc=<4.3188 8.5297 -3.5471>, rot=<-0.0748 -0.5282 0.8446 -0.0444>}
 dyn_hair_e_l_parent {parent=dyn_hair_e_l_base, loc=<0 0 0>, rot=<0 0 0 1>}
 dyn_hair_e_l_01 {parent=dyn_hair_e_l_parent, loc=<3.7058 0.8234 -0.5419>, rot=<-0.0456 0.134 0.0774 -0.9869>}
 dyn_hair_e_l_02 {parent=dyn_hair_e_l_parent, loc=<3.8018 -1.2346 -0.0645>, rot=<0.1064 0.1073 -0.0793 -0.9853>}
 dyn_hair_f_r_base {parent=dyn_hair_offset, loc=<-5.6571 -5.8809 8.2704>, rot=<-0.0188 -0.9393 -0.2954 0.1734>}
 dyn_hair_f_r_parent {parent=dyn_hair_f_r_base, loc=<0 0 0>, rot=<0 0 0 1>}
 dyn_hair_f_r_01 {parent=dyn_hair_f_r_parent, loc=<4.0823 -0.2301 -0.7972>, rot=<-0.0478 -0.0076 0.0226 -0.9986>}
 dyn_hair_f_r_02 {parent=dyn_hair_f_r_parent, loc=<3.7632 2.8437 0.2036>, rot=<-0.1739 0.1169 -0.2681 -0.9403>}
 dyn_hair_g_l_base {parent=dyn_hair_offset, loc=<0.3202 -9.2692 -0.0869>, rot=<-0.1299 0.6958 0.6959 -0.1213>}
 dyn_hair_g_l_parent {parent=dyn_hair_g_l_base, loc=<0 0 0>, rot=<0 0 0 1>}
 dyn_hair_g_l_01 {parent=dyn_hair_g_l_parent, loc=<5.128 -0.0244 0.7163>, rot=<-0.0419 -0.0874 -0.002 -0.9953>}
 dyn_hair_g_l_02 {parent=dyn_hair_g_l_01, loc=<4.6816 -0.16 1.6193>, rot=<-0.045 0.0873 -0.0056 -0.9952>}
 dyn_hair_h_l_base {parent=dyn_hair_offset, loc=<-0.1532 -5.8971 -5.6161>, rot=<-0.162 0.3762 0.9096 -0.0702>}
 dyn_hair_h_l_parent {parent=dyn_hair_h_l_base, loc=<0 0 0>, rot=<0 0 0 1>}
 dyn_hair_h_l_01 {parent=dyn_hair_h_l_parent, loc=<5.2703 -0.0927 0.0816>, rot=<0 0 0 1>}
 dyn_hair_h_l_02 {parent=dyn_hair_h_l_parent, loc=<4.4874 -2.3097 0.4202>, rot=<0.0004 -0.0228 0.4756 -0.8793>}
 dyn_hair_h_l_03 {parent=dyn_hair_h_l_02, loc=<3.2549 1.0761 0.0794>, rot=<-0.054 0.0968 -0.0572 -0.9922>}
 dyn_hair_i_l_base {parent=dyn_hair_offset, loc=<-0.455 -3.0459 -8.5824>, rot=<-0.4918 -0.1524 0.8532 0.0833>}
 dyn_hair_i_l_parent {parent=dyn_hair_i_l_base, loc=<0 0 0>, rot=<0 0 0 1>}
 dyn_hair_b_r_base {parent=dyn_hair_offset, loc=<-5.0282 -5.1698 10.2854>, rot=<-0.0733 0.8128 0.0975 -0.5697>}
 dyn_hair_b_r_parent {parent=dyn_hair_b_r_base, loc=<0 0 0>, rot=<0 0 0 1>}
 dyn_scarf_base {parent=dyn_hair_offset, loc=<10.5065 -6.3817 -0.1653>, rot=<-0.3291 0.6259 0.6259 -0.3291>}
 dyn_scarf_parent {parent=dyn_scarf_base, loc=<0 0 0>, rot=<0 0 0 1>}
 dyn_scarf_ribbon_top_base {parent=dyn_hair_offset, loc=<6.1335 -1.1861 -9.3376>, rot=<-0.4909 -0.3547 0.7386 -0.296>}
 dyn_scarf_ribbon_top_parent {parent=dyn_scarf_ribbon_top_base, loc=<0 0 0>, rot=<0 0 0 1>}
 dyn_scarf_ribbon_btm_base {parent=dyn_hair_offset, loc=<5.7101 -1.2593 -9.1131>, rot=<-0.4893 -0.1213 0.8482 -0.1627>}
 dyn_scarf_ribbon_btm_parent {parent=dyn_scarf_ribbon_btm_base, loc=<0 0 0>, rot=<0 0 0 1>}
 attach_backpack {parent=spine_05, loc=<0 -8.35 0>, rot=<-0.0093 -0.707 0.0093 -0.707>}
 clavicle_l {parent=spine_05, loc=<5.4831 -0.1053 -3.2946>, rot=<0.0465 -0.7289 -0.009 -0.6829>}
 upperarm_l {parent=clavicle_l, loc=<9.8658 -1.2136 -0.4105>, rot=<0.0164 -0.3245 -0.0269 -0.9454>}
 lowerarm_l {parent=upperarm_l, loc=<26.465 -0.8345 0.259>, rot=<0.0036 0.0161 -0.2298 -0.9731>}
 hand_l {parent=lowerarm_l, loc=<22.7355 -0.3203 -0.6105>, rot=<-0.6784 -0.0223 0.0735 -0.7307>}
 thumb_01_l {parent=hand_l, loc=<1.3118 -0.3191 -2.459>, rot=<0.6546 -0.4042 0.0664 -0.6354>}
 thumb_02_l {parent=thumb_01_l, loc=<5.9336 -0.0448 0.0401>, rot=<-0.0037 0.0213 0.298 -0.9543>}
 thumb_03_l {parent=thumb_02_l, loc=<2.2223 0.3682 -0.0907>, rot=<0.0118 -0.013 0.0066 -0.9998>}
 index_metacarpal_l {parent=hand_l, loc=<1.6535 0.0183 -2.96>, rot=<0.0158 -0.0124 -0.0173 -0.9996>}
 index_01_l {parent=index_metacarpal_l, loc=<8.3008 0.0357 -0.0524>, rot=<-0.0044 0.0485 0.0924 -0.9945>}
 index_02_l {parent=index_01_l, loc=<3.3587 -0.3004 0.0452>, rot=<-0.002 0.0027 0.3225 -0.9466>}
 index_03_l {parent=index_02_l, loc=<2.5578 -0.123 -0.0252>, rot=<-0.002 0.0067 0.2558 -0.9667>}
 middle_metacarpal_l {parent=hand_l, loc=<2.1856 -0.1725 -0.5984>, rot=<0.006 0.0168 -0.0329 -0.9993>}
 middle_01_l {parent=middle_metacarpal_l, loc=<7.3275 -0.2447 0.0129>, rot=<-0.0208 0.0042 0.0858 -0.9961>}
 middle_02_l {parent=middle_01_l, loc=<4.0123 0.1401 0.1175>, rot=<-0.0218 0.0258 0.3512 -0.9357>}
 middle_03_l {parent=middle_02_l, loc=<2.6077 0.0534 0.1373>, rot=<0.0082 0.0188 0.1708 -0.9851>}
 ring_metacarpal_l {parent=hand_l, loc=<2.2074 -0.3625 1.3638>, rot=<-0.0284 0.0722 -0.0458 -0.9959>}
 ring_01_l {parent=ring_metacarpal_l, loc=<7.0295 -0.6197 -0.1446>, rot=<-0.0489 -0.0252 0.1892 -0.9804>}
 ring_02_l {parent=ring_01_l, loc=<3.6174 0.4647 -0.1027>, rot=<0 0 0.2806 -0.9598>}
 ring_03_l {parent=ring_02_l, loc=<2.4152 0.114 0.0199>, rot=<0 0 0.1189 -0.9929>}
 pinky_metacarpal_l {parent=hand_l, loc=<1.7616 -0.9753 3.3814>, rot=<-0.0397 0.0814 -0.069 -0.9935>}
 pinky_01_l {parent=pinky_metacarpal_l, loc=<7.1709 -0.4831 -0.1094>, rot=<-0.0654 -0.0035 0.2242 -0.9723>}
 pinky_02_l {parent=pinky_01_l, loc=<2.8274 0.0893 -0.1043>, rot=<0.0037 0.0011 0.2854 -0.9584>}
 pinky_03_l {parent=pinky_02_l, loc=<2.2046 0.0307 -0.0078>, rot=<-0.0002 0.0098 0.1445 -0.9895>}
 weapon_l {parent=hand_l, loc=<-0.1974 -3.4111 0.3811>, rot=<0 0 0 1>}
 hand_attach_l {parent=hand_l, loc=<8.5 -3 0>, rot=<0.7071 0.7071 0 0>}
 lowerarm_twist_02_l {parent=lowerarm_l, loc=<10.6707 0 0>, rot=<0 0 0 -1>}
 lowerarm_twist_01_l {parent=lowerarm_l, loc=<18.316 0 0>, rot=<0 0 0 -1>}
 upperarm_twist_01_l {parent=upperarm_l, loc=<7.2135 -0.1134 0>, rot=<0 0 0 1>}
 upperarm_twist_02_l {parent=upperarm_l, loc=<19.7998 -0.1134 0>, rot=<0 0 0 1>}
 clavicle_r {parent=spine_05, loc=<5.4831 -0.1053 3.2946>, rot=<0.7289 0.0465 -0.6829 0.009>}
 upperarm_r {parent=clavicle_r, loc=<-9.8658 1.2136 0.4105>, rot=<0.0164 -0.3245 -0.0269 -0.9454>}
 lowerarm_r {parent=upperarm_r, loc=<-26.465 0.8345 -0.259>, rot=<0.0036 0.0161 -0.2298 -0.9731>}
 hand_r {parent=lowerarm_r, loc=<-22.7355 0.3203 0.6106>, rot=<-0.6784 -0.0223 0.0735 -0.7307>}
 thumb_01_r {parent=hand_r, loc=<-1.3118 0.3191 2.459>, rot=<0.6546 -0.4042 0.0664 -0.6354>}
 thumb_02_r {parent=thumb_01_r, loc=<-5.9336 0.0448 -0.0401>, rot=<-0.0037 0.0213 0.298 -0.9543>}
 thumb_03_r {parent=thumb_02_r, loc=<-2.2223 -0.3682 0.0907>, rot=<0.0118 -0.013 0.0066 -0.9998>}
 index_metacarpal_r {parent=hand_r, loc=<-1.6535 -0.0183 2.96>, rot=<0.0158 -0.0124 -0.0173 -0.9996>}
 index_01_r {parent=index_metacarpal_r, loc=<-8.3008 -0.0357 0.0524>, rot=<-0.0044 0.0485 0.0924 -0.9945>}
 index_02_r {parent=index_01_r, loc=<-3.3587 0.3004 -0.0452>, rot=<-0.002 0.0027 0.3225 -0.9466>}
 index_03_r {parent=index_02_r, loc=<-2.5578 0.123 0.0252>, rot=<-0.002 0.0067 0.2558 -0.9667>}
 middle_metacarpal_r {parent=hand_r, loc=<-2.1856 0.1725 0.5984>, rot=<0.006 0.0168 -0.0329 -0.9993>}
 middle_01_r {parent=middle_metacarpal_r, loc=<-7.3275 0.2447 -0.0129>, rot=<-0.0208 0.0042 0.0858 -0.9961>}
 middle_02_r {parent=middle_01_r, loc=<-4.0123 -0.1401 -0.1175>, rot=<-0.0218 0.0258 0.3512 -0.9357>}
 middle_03_r {parent=middle_02_r, loc=<-2.6077 -0.0534 -0.1373>, rot=<0.0082 0.0188 0.1708 -0.9851>}
 ring_metacarpal_r {parent=hand_r, loc=<-2.2074 0.3625 -1.3638>, rot=<-0.0284 0.0722 -0.0458 -0.9959>}
 ring_01_r {parent=ring_metacarpal_r, loc=<-7.0295 0.6197 0.1446>, rot=<-0.0489 -0.0252 0.1892 -0.9804>}
 ring_02_r {parent=ring_01_r, loc=<-3.6174 -0.4647 0.1027>, rot=<0 0 0.2806 -0.9598>}
 ring_03_r {parent=ring_02_r, loc=<-2.4152 -0.114 -0.0199>, rot=<0 0 0.1189 -0.9929>}
 pinky_metacarpal_r {parent=hand_r, loc=<-1.7616 0.9753 -3.3814>, rot=<-0.0397 0.0814 -0.069 -0.9935>}
 pinky_01_r {parent=pinky_metacarpal_r, loc=<-7.1709 0.4831 0.1094>, rot=<-0.0654 -0.0035 0.2242 -0.9723>}
 pinky_02_r {parent=pinky_01_r, loc=<-2.8274 -0.0893 0.1043>, rot=<0.0037 0.0011 0.2854 -0.9584>}
 pinky_03_r {parent=pinky_02_r, loc=<-2.2046 -0.0307 0.0078>, rot=<-0.0002 0.0098 0.1445 -0.9895>}
 weapon_r {parent=hand_r, loc=<0.1974 3.4111 -0.3811>, rot=<0 0 0 1>}
 hand_attach_r {parent=hand_r, loc=<-8.5001 2.9999 0>, rot=<0 0 -0.7071 -0.7071>}
 lowerarm_twist_02_r {parent=lowerarm_r, loc=<-10.6707 0 0>, rot=<0 0 0 1>}
 dyn_sleeve_inner_base_r {parent=lowerarm_twist_02_r, loc=<-0.387 -4.7691 3.7447>, rot=<-0.0358 -0.8555 -0.497 0.1407>}
 dyn_sleeve_outter_base_r {parent=lowerarm_twist_02_r, loc=<0.3331 5.4304 -1.2819>, rot=<-0.0196 -0.4973 0.8641 0.0746>}
 dyn_sleeve_bottom_base_r {parent=lowerarm_twist_02_r, loc=<1.4716 2.8177 7.192>, rot=<-0.0523 -0.9425 0.2874 0.1624>}
 dyn_sleeve_top_base_r {parent=lowerarm_twist_02_r, loc=<-1.3254 -3.0469 -5.6698>, rot=<-0.0926 0.1575 0.9831 0.0101>}
 lowerarm_twist_01_r {parent=lowerarm_r, loc=<-18.316 0 0>, rot=<0 0 0 1>}
 upperarm_twist_01_r {parent=upperarm_r, loc=<-7.2135 0.1134 0>, rot=<0 0 0 -1>}
 upperarm_twist_02_r {parent=upperarm_r, loc=<-19.7998 0.1134 0>, rot=<0 0 0 -1>}
 dyn_string_base_r {parent=spine_05, loc=<7.6622 8.027 3.5424>, rot=<-0.0461 -0.8016 0.5855 0.1121>}
 dyn_string_parent_r {parent=dyn_string_base_r, loc=<0 0 0>, rot=<0 0 0 1>}
 dyn_string_01_r {parent=dyn_string_parent_r, loc=<2.0854 -0.0142 0.0604>, rot=<-0.0467 -0.0191 -0.0331 -0.9982>}
 dyn_string_02_r {parent=dyn_string_01_r, loc=<4.3443 0.0298 -0.3677>, rot=<-0.0157 -0.1045 -0.0077 -0.9944>}
 dyn_string_03_r {parent=dyn_string_02_r, loc=<4.7385 0.1142 -0.2391>, rot=<0 -0.0349 0 -0.9994>}
 dyn_string_base_l {parent=spine_05, loc=<7.6182 8.1165 -3.4897>, rot=<-0.0461 -0.8016 0.5855 0.1121>}
 dyn_string_parent_l {parent=dyn_string_base_l, loc=<0 0 0>, rot=<0 0 0 1>}
 dyn_string_01_l {parent=dyn_string_parent_l, loc=<1.9682 0.4413 0.2709>, rot=<-0.0467 -0.0191 -0.0331 -0.9982>}
 dyn_string_02_l {parent=dyn_string_01_l, loc=<4.2865 0.7081 -0.2227>, rot=<-0.0707 -0.1151 -0.0792 -0.9877>}
 dyn_string_03_l {parent=dyn_string_02_l, loc=<4.7447 0.0851 -0.0635>, rot=<-0.0316 -0.0334 -0.0144 -0.9988>}
 dyn_zipper_neck_base {parent=spine_05, loc=<4.9812 8.03 -0.0002>, rot=<-0.0948 -0.7007 0.7007 0.0948>}
 dyn_zipper_neck_parent {parent=dyn_zipper_neck_base, loc=<0 0 0>, rot=<0 0 0 1>}
 dyn_collar_frnt_base_l {parent=spine_05, loc=<11.1225 6.1928 -6.6337>, rot=<-0.3899 0.0911 -0.1508 -0.9039>}
 dyn_collar_frnt_parent_l {parent=dyn_collar_frnt_base_l, loc=<0 0 0>, rot=<0 0 0 1>}
 dyn_collar_frnt_base_r {parent=spine_05, loc=<11.1227 6.1928 6.6337>, rot=<0.8843 0.1607 -0.0436 0.4362>}
 dyn_collar_frnt_parent_r {parent=dyn_collar_frnt_base_r, loc=<0 0 0>, rot=<0 0 0 1>}
 dyn_zipper_stomach_base {parent=spine_02, loc=<-2.6237 10.1416 -0.0002>, rot=<0.0329 -0.8248 0.564 -0.0225>}
 dyn_zipper_stomach_parent {parent=dyn_zipper_stomach_base, loc=<0 0 0>, rot=<0 0 0 1>}
 dyn_clock_chain_base {parent=spine_02, loc=<-3.9296 -0.4672 11.6107>, rot=<0.0905 -0.9253 -0.1334 0.3434>}
 dyn_clock_chain_parent {parent=dyn_clock_chain_base, loc=<0 0 0>, rot=<0 0 0 -1>}
 dyn_clock_chain_01 {parent=dyn_clock_chain_parent, loc=<1.6965 -0.0236 -0.0178>, rot=<-0.0557 0.0529 0.0068 -0.997>}
 dyn_clock_chain_02 {parent=dyn_clock_chain_01, loc=<2.6963 -0.1474 0.0393>, rot=<0.0148 0.0358 0.0043 -0.9992>}
 dyn_clock_chain_03 {parent=dyn_clock_chain_02, loc=<2.8861 0.0237 0.1217>, rot=<0.0308 0.0368 0.0114 -0.9988>}
 dfrm_clock_hour {parent=dyn_clock_chain_03, loc=<10.012 0.8865 -1.3202>, rot=<0.5755 -0.5755 -0.4046 0.4171>}
 dfrm_clock_minute {parent=dyn_clock_chain_03, loc=<10.012 0.8865 -1.3202>, rot=<0.5755 -0.5755 -0.4046 0.4171>}
 dfrm_clock_second {parent=dyn_clock_chain_03, loc=<10.012 0.8865 -1.3202>, rot=<0.5755 -0.5755 -0.4046 0.4171>}
 thigh_l {parent=pelvis, loc=<-1.0855 -0.2109 -8.7053>, rot=<0.0009 0.0355 0.0155 -0.9993>}
 calf_l {parent=thigh_l, loc=<-42.5284 -0.0019 -0.7297>, rot=<0.0006 -0.0161 -0.0391 -0.9991>}
 foot_l {parent=calf_l, loc=<-46.7401 -2.5014 -0.592>, rot=<0.0427 -0.0238 0.0353 -0.9982>}
 ball_l {parent=foot_l, loc=<-3.5658 10.8066 -0.1544>, rot=<0.0001 -0.0001 -0.7063 -0.7079>}
 calf_twist_02_l {parent=calf_l, loc=<-19.1546 -1.7943 -0.2638>, rot=<0.0237 0.0095 -0.0035 -0.9997>}
 dyn_zipper_leg_base_l {parent=calf_twist_02_l, loc=<-6.4587 -0.8282 -5.4523>, rot=<0.0299 -0.0257 0.9982 0.0445>}
 dyn_zipper_leg_parent_l {parent=dyn_zipper_leg_base_l, loc=<0 0 0>, rot=<0 0 0 1>}
 calf_twist_01_l {parent=calf_l, loc=<-35.8483 -1.9197 -0.5794>, rot=<0.0237 0.0095 -0.0035 -0.9997>}
 thigh_twist_01_l {parent=thigh_l, loc=<-27.952 0 0>, rot=<0.0013 0 0 -1>}
 dyn_skirt_correction_base_l {parent=thigh_l, loc=<-15.5879 -0.2006 -0.2049>, rot=<-0.4809 0.483 0.5188 -0.5161>}
 dyn_skirt_correction_parent_l {parent=dyn_skirt_correction_base_l, loc=<0 0 0>, rot=<0 0 0 1>}
 dyn_chain_corrective {parent=thigh_l, loc=<-13.0366 -0.9779 0.1673>, rot=<0.483 0.4809 -0.5161 -0.5188>}
 dfrm_thigh_l {parent=thigh_l, loc=<0 0 0>, rot=<-0.0086 0 1 0>}
 dfrm_back_thigh_l {parent=dfrm_thigh_l, loc=<0 0 0>, rot=<0 0 0 1>}
 dfrm_front_thigh_l {parent=dfrm_thigh_l, loc=<0 0 0>, rot=<0 0 0 1>}
 thigh_r {parent=pelvis, loc=<-1.0855 -0.2109 8.7053>, rot=<0.0355 -0.0009 0.9993 0.0155>}
 calf_r {parent=thigh_r, loc=<42.5284 0.0019 0.7297>, rot=<0.0006 -0.0161 -0.0391 -0.9991>}
 foot_r {parent=calf_r, loc=<46.7401 2.5014 0.592>, rot=<0.0427 -0.0238 0.0353 -0.9982>}
 ball_r {parent=foot_r, loc=<3.5658 -10.8066 0.1544>, rot=<0.0001 -0.0001 -0.7063 -0.7079>}
 calf_twist_02_r {parent=calf_r, loc=<19.1546 1.7943 0.2638>, rot=<0.0237 0.0095 -0.0035 -0.9997>}
 calf_twist_01_r {parent=calf_r, loc=<35.8483 1.9197 0.5794>, rot=<0.0237 0.0095 -0.0035 -0.9997>}
 thigh_twist_01_r {parent=thigh_r, loc=<27.952 0 0>, rot=<0.0013 0 0 -1>}
 dyn_skirt_correction_base_r {parent=thigh_r, loc=<15.5878 0.2006 0.2049>, rot=<0.5188 0.5161 0.4809 0.483>}
 dyn_skirt_correction_parent_r {parent=dyn_skirt_correction_base_r, loc=<0 0 0>, rot=<0 0 0 1>}
 dfrm_thigh_r {parent=thigh_r, loc=<0 0 0>, rot=<0 0.0086 0 -1>}
 dfrm_back_thigh_r {parent=dfrm_thigh_r, loc=<0 0 0>, rot=<0 0 0 1>}
 dfrm_front_thigh_r {parent=dfrm_thigh_r, loc=<0 0 0>, rot=<0 0 0 1>}
 dyn_belt_side_base_l {parent=pelvis, loc=<-1.4967 -0.4762 -15.9334>, rot=<0.707 -0.0127 -0.707 -0.0127>}
 dyn_belt_side_parent_l {parent=dyn_belt_side_base_l, loc=<0 0 0>, rot=<0 0 0 1>}
 dyn_belt_tassel_j_base {parent=pelvis, loc=<-3.3649 5.0293 -14.4069>, rot=<-0.0881 -0.4037 0.9093 0.0489>}
 dyn_belt_tassel_j_mstr {parent=dyn_belt_tassel_j_base, loc=<0 0 0>, rot=<0 0 0 1>}
 dyn_belt_tassel_j_parent {parent=dyn_belt_tassel_j_mstr, loc=<0 0 0>, rot=<0 0 0 1>}
 dyn_belt_tassel_i_base {parent=pelvis, loc=<-3.349 1.0371 -16.5403>, rot=<-0.0718 -0.1203 0.99 0.0141>}
 dyn_belt_tassel_i_parent {parent=dyn_belt_tassel_i_base, loc=<0 0 0>, rot=<0 0 0 1>}
 dyn_belt_tassel_h_base {parent=pelvis, loc=<-3.0073 -2.6623 -16.4151>, rot=<-0.0774 0.0882 0.9929 -0.019>}
 dyn_belt_tassel_h_parent {parent=dyn_belt_tassel_h_base, loc=<0 0 0>, rot=<0 0 0 1>}
 dyn_belt_tassel_f_base {parent=pelvis, loc=<-0.8137 -9.1932 -12.0042>, rot=<-0.0757 0.451 0.8842 -0.0953>}
 dyn_belt_tassel_f_parent {parent=dyn_belt_tassel_f_base, loc=<0 0 0>, rot=<0 0 0 1>}
 dyn_belt_tassel_g_base {parent=pelvis, loc=<-2.115 -6.2308 -14.7973>, rot=<-0.065 0.259 0.963 -0.0352>}
 dyn_belt_tassel_g_parent {parent=dyn_belt_tassel_g_base, loc=<0 0 0>, rot=<0 0 0 1>}
 dyn_belt_frnt_base_l {parent=pelvis, loc=<-0.0808 7.4515 -7.1931>, rot=<0.5574 0.4351 -0.5415 -0.4548>}
 dyn_belt_frnt_parent_l {parent=dyn_belt_frnt_base_l, loc=<0 0 0>, rot=<0 0 0 1>}
 dyn_belt_tassel_o_base {parent=pelvis, loc=<0.4527 8.809 -0.1689>, rot=<-0.0226 -0.731 0.6815 0.0268>}
 dyn_belt_tassel_o_parent {parent=dyn_belt_tassel_o_base, loc=<0 0 0>, rot=<0 0 0 1>}
 dyn_belt_tassel_p_base {parent=pelvis, loc=<-1.3223 8.3748 -5.185>, rot=<-0.0122 -0.6819 0.7312 0.0131>}
 dyn_belt_tassel_p_parent {parent=dyn_belt_tassel_p_base, loc=<0 0 0>, rot=<0 0 0 1>}
 dyn_belt_bk_base_l {parent=pelvis, loc=<3.4016 -10.2868 -6.1713>, rot=<-0.6361 0.3615 0.4268 -0.5315>}
 dyn_belt_bk_parent_l {parent=dyn_belt_bk_base_l, loc=<0 0 0>, rot=<0 0 0 1>}
 dyn_belt_tassel_e_base {parent=pelvis, loc=<2.5014 -11.465 -4.3693>, rot=<-0.1081 0.6715 0.7205 -0.1354>}
 dyn_belt_tassel_e_parent {parent=dyn_belt_tassel_e_base, loc=<0 0 0>, rot=<0 0 0 1>}
 dyn_belt_tassel_a_base {parent=pelvis, loc=<8.8394 -4.6999 10.5271>, rot=<-0.0031 -0.934 -0.3243 0.1501>}
 dyn_belt_tassel_a_parent {parent=dyn_belt_tassel_a_base, loc=<0 0 0>, rot=<0 0 0 1>}
 dyn_belt_tassel_b_base {parent=pelvis, loc=<7.5718 -7.8471 7.5801>, rot=<0.0196 -0.8463 -0.5015 0.1787>}
 dyn_belt_tassel_b_parent {parent=dyn_belt_tassel_b_base, loc=<0 0 0>, rot=<0 0 0 1>}
 dyn_belt_tassel_c_base {parent=pelvis, loc=<6.0692 -9.6619 4.0538>, rot=<0.0955 -0.7591 -0.6166 0.1857>}
 dyn_belt_tassel_c_parent {parent=dyn_belt_tassel_c_base, loc=<0 0 0>, rot=<0 0 0 1>}
 dyn_belt_tassel_k_base {parent=pelvis, loc=<8.7158 1.7095 12.337>, rot=<0.0836 -0.9809 0.0656 0.163>}
 dyn_belt_tassel_k_parent {parent=dyn_belt_tassel_k_base, loc=<0 0 0>, rot=<0 0 0 1>}
 dyn_belt_tassel_l_base {parent=pelvis, loc=<7.1111 5.022 11.1163>, rot=<0.0169 -0.9331 0.3373 0.1232>}
 dyn_belt_tassel_l_parent {parent=dyn_belt_tassel_l_base, loc=<0 0 0>, rot=<0 0 0 1>}
 dyn_belt_tassel_m_base {parent=pelvis, loc=<4.555 7.5228 7.8828>, rot=<0.0207 -0.8226 0.5639 0.0702>}
 dyn_belt_tassel_m_parent {parent=dyn_belt_tassel_m_base, loc=<0 0 0>, rot=<0 0 0 1>}
 dyn_belt_tassel_n_base {parent=pelvis, loc=<2.368 8.5762 4.066>, rot=<-0.0041 -0.7431 0.6687 0.0268>}
 dyn_belt_tassel_n_parent {parent=dyn_belt_tassel_n_base, loc=<0 0 0>, rot=<0 0 0 1>}
 dyn_belt_tassel_d_base {parent=pelvis, loc=<4.3999 -10.8516 0.0899>, rot=<0.1261 -0.7293 -0.6572 0.1428>}
 dyn_belt_tassel_d_parent {parent=dyn_belt_tassel_d_base, loc=<0 0 0>, rot=<0 0 0 1>}
 const_chain_frnt_base {parent=pelvis, loc=<-2.7373 7.5077 -9.8538>, rot=<-0.1923 0.1609 0.9677 0.0257>}
 const_mstr_chain_frnt {parent=const_chain_frnt_base, loc=<0 0 0>, rot=<0 0 0 1>}
 dyn_mstr_chain_frnt_l {parent=const_mstr_chain_frnt, loc=<4.7783 0 0>, rot=<-0.1055 -0.1084 0.0018 -0.9885>}
 dyn_chain_frnt_01_l {parent=dyn_mstr_chain_frnt_l, loc=<0 0 0>, rot=<0.0839 0 0 -0.9965>}
 dyn_chain_frnt_02_l {parent=dyn_chain_frnt_01_l, loc=<2.4057 0 0>, rot=<-0.0074 0.0114 -0.0411 -0.9991>}
 dyn_chain_frnt_03_l {parent=dyn_chain_frnt_02_l, loc=<2.0601 0 0>, rot=<0.0009 0.0088 -0.0256 -0.9996>}
 dyn_chain_frnt_04_l {parent=dyn_chain_frnt_03_l, loc=<1.8831 0 0>, rot=<-0.015 -0.0129 -0.0483 -0.9986>}
 dyn_chain_frnt_05_l {parent=dyn_chain_frnt_04_l, loc=<1.9059 0 0>, rot=<-0.0024 -0.0229 -0.0676 -0.9975>}
 dyn_chain_frnt_06_l {parent=dyn_chain_frnt_05_l, loc=<1.8261 0 0>, rot=<-0.016 -0.0432 -0.0711 -0.9964>}
 dyn_chain_frnt_07_l {parent=dyn_chain_frnt_06_l, loc=<1.8239 0 0>, rot=<0.0171 -0.0374 -0.0752 -0.9963>}
 dyn_chain_frnt_08_l {parent=dyn_chain_frnt_07_l, loc=<1.7306 0 0>, rot=<0.011 -0.0355 -0.0952 -0.9948>}
 dyn_chain_frnt_09_l {parent=dyn_chain_frnt_08_l, loc=<1.6639 0 0>, rot=<-0.0009 -0.0353 -0.1278 -0.9912>}
 dyn_chain_frnt_010_l {parent=dyn_chain_frnt_09_l, loc=<1.6907 0 0>, rot=<-0.0049 0.0087 -0.118 -0.993>}
 dyn_chain_frnt_011_l {parent=dyn_chain_frnt_010_l, loc=<1.6609 0 0>, rot=<-0.0656 -0.044 -0.0742 -0.9941>}
 const_chain_bk_base {parent=pelvis, loc=<0.8462 -11.0647 -8.3299>, rot=<-0.0923 0.5797 0.8074 -0.0591>}
 const_chain_bk_parent {parent=const_chain_bk_base, loc=<0.7334 0 0>, rot=<-0.7033 -0.1123 -0.0311 -0.7013>}
 dyn_chain_attach_bk_02 {parent=const_chain_bk_parent, loc=<2.7785 0 0>, rot=<0.065 0.2577 -0.1004 -0.9588>}
 dyn_mstr_chain_back_l {parent=dyn_chain_attach_bk_02, loc=<1.2692 -0.0559 -0.0022>, rot=<-0.0972 0.099 -0.0024 -0.9903>}
 dyn_chain_back_01_l {parent=dyn_mstr_chain_back_l, loc=<0 0 0>, rot=<0.0022 0 0 -1>}
 dyn_chain_back_02_l {parent=dyn_chain_back_01_l, loc=<2.0222 0 0>, rot=<-0.0168 -0.0144 -0.0051 -0.9997>}
 dyn_chain_back_03_l {parent=dyn_chain_back_02_l, loc=<1.7498 0 0>, rot=<-0.0215 0.0003 0.0034 -0.9998>}
 dyn_chain_back_04_l {parent=dyn_chain_back_03_l, loc=<1.8231 0 0>, rot=<0.0106 0.0186 -0.0152 -0.9997>}
 dyn_chain_back_05_l {parent=dyn_chain_back_04_l, loc=<1.8673 0 0>, rot=<-0.0109 0.0275 -0.0321 -0.999>}
 dyn_chain_back_06_l {parent=dyn_chain_back_05_l, loc=<1.9742 0 0>, rot=<0.0324 0.0118 -0.0388 -0.9987>}
 dyn_chain_back_07_l {parent=dyn_chain_back_06_l, loc=<1.9272 0 0>, rot=<0.0281 0.0094 -0.0478 -0.9984>}
 dyn_chain_back_08_l {parent=dyn_chain_back_07_l, loc=<1.8136 0 0>, rot=<0.043 0.0147 -0.0546 -0.9975>}
 dyn_chain_back_09_l {parent=dyn_chain_back_08_l, loc=<1.7406 0 0>, rot=<0.0105 0.0222 -0.0445 -0.9987>}
 dyn_chain_back_010_l {parent=dyn_chain_back_09_l, loc=<1.6266 0 0>, rot=<-0.0317 0.0257 -0.0582 -0.9975>}
 dyn_chain_back_011_l {parent=dyn_chain_back_010_l, loc=<1.6362 0 0>, rot=<0.0475 0.0733 -0.1146 -0.9896>}
 dyn_chain_back_012_l {parent=dyn_chain_back_011_l, loc=<1.7867 0 0>, rot=<-0.0444 0.0738 -0.1116 -0.99>}
 dyn_chain_back_013_l {parent=dyn_chain_back_012_l, loc=<1.6748 0 0>, rot=<-0.0187 0.0694 -0.1005 -0.9923>}
 dyn_chain_back_014_l {parent=dyn_chain_back_013_l, loc=<2.194 0 0>, rot=<0.0092 0.0454 -0.0734 -0.9962>}
 dfrm_aim_up_thigh_l {parent=pelvis, loc=<-0.5577 -0.2315 -14.682>, rot=<-0.044 0.001 0.9989 0.0155>}
 dfrm_aim_up_thigh_r {parent=pelvis, loc=<-0.6457 -0.228 13.6859>, rot=<0.044 -0.001 0.9989 0.0155>}
 dfrm_mstr_skirt_back_mid {parent=pelvis, loc=<-22.278 -7.4968 0>, rot=<0 0 0 1>}
 dfrm_skirt_back_mid {parent=dfrm_mstr_skirt_back_mid, loc=<0 0 0>, rot=<0 0 0 1>}
 dyn_mstr_chain_sim {parent=pelvis, loc=<0.1512 -2.4954 -14.714>, rot=<-0.0797 0.691 0.7109 0.1037>}
 dyn_chain_sim_01 {parent=dyn_mstr_chain_sim, loc=<0 0 0>, rot=<0 0 0 1>}
 dyn_chain_sim_02 {parent=dyn_chain_sim_01, loc=<10.0185 0 0>, rot=<0 0 -0.079 -0.9969>}
 dyn_belt_frnt_base_mid {parent=pelvis, loc=<2.4235 8.2358 -0.1724>, rot=<0.5 0.5 -0.5 -0.5>}
 dyn_belt_frnt_parent_mid {parent=dyn_belt_frnt_base_mid, loc=<0 0 0>, rot=<0 0 0 1>}
]
variables [game_time_in_seconds {type=double, default=0}]
