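# Physics asset for the hanging clock-on-chain accessory of the Silent Tempo medium female cosmetic body.
# A five-bone dyn_clock_chain runs from a small sphere base through thin capsules to a large flat box on the last bone, likely the clock.
# Constraints link each bone to the previous one with tight, asymmetric swing and narrow twist limits.
# Unconstrained capsules on the pelvis and right thigh give the chain body shapes to collide with.
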PhysicsAsset /BRCosmetics/Characters/Player/Female/Medium/Bodies/F_MED_SilentTempo/Meshes/F_MED_SilentTempo_Clock_Physics
bodies {
 dyn_clock_chain_01 [{shapeType=Capsule, length=2.532, radius=0.9889, shapeName=dyn_clock_chain_01_capsule, center=<2.2321 -0.045 0.0081>, rotation=<rot 90 0 0>}]
 dyn_clock_chain_02 [{shapeType=Capsule, length=2.8486, radius=0.6229, shapeName=dyn_clock_chain_02_capsule, center=<2.0444 0.0304 0.0789>, rotation=<rot 90 0 0>}]
 dyn_clock_chain_03 [{
   shapeType=Box
   extentX=21.8316
   extentY=24.0138
   extentZ=2.9774
   shapeName=dyn_clock_chain_03_box
   center=<10.3725 0.3216 1.577>
   rotation=<rot 18.6825 0.0003 0.0001>
  }]
 dyn_clock_chain_base [{shapeType=Sphere, radius=0.6834, shapeName=dyn_clock_chain_base_sphere}]
 dyn_clock_chain_parent [{
   shapeType=Capsule
   length=1.6728
   radius=0.5
   shapeName=dyn_clock_chain_parent_capsule
   center=<2.2173 -0.0289 -0.033>
   rotation=<rot 90 0 0>
  }]
 pelvis [{shapeType=Capsule, length=11.424, radius=11.2245, shapeName=pelvis_capsule, center=<-7.2265 -0.6576 0.1556>}]
 thigh_r [{shapeType=Capsule, length=21.0723, radius=8.9277, shapeName=thigh_r_capsule, center=<21.8411 0 -0.7529>, rotation=<rot 89.5077 179.9999 180>}]
}
constraints [
 {bone1Name=dyn_clock_chain_parent, bone2Name=dyn_clock_chain_base, swing1LimitDegrees=20, swing2LimitDegrees=10, twistLimitDegrees=5}
 {bone1Name=dyn_clock_chain_01, bone2Name=dyn_clock_chain_parent, swing1LimitDegrees=20, swing2LimitDegrees=10, twistLimitDegrees=3}
 {bone1Name=dyn_clock_chain_02, bone2Name=dyn_clock_chain_01, swing1LimitDegrees=20, swing2LimitDegrees=10, twistLimitDegrees=3}
 {bone1Name=dyn_clock_chain_03, bone2Name=dyn_clock_chain_02, swing1LimitDegrees=20, swing2LimitDegrees=10, twistLimitDegrees=3}
]
