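# Physics asset for the SilentTempo female medium body mesh, driving left and right dangling string bone chains.
# Each chain runs from a small sphere base through parent and three capsule segments, linked by swing-limited constraints with very tight twist.
# The upper left-side joints allow wider swing than the rest.
# A capsule and box on the upper spine bones have no constraints and form torso shapes, likely for the strings to collide against.
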PhysicsAsset /BRCosmetics/Characters/Player/Female/Medium/Bodies/F_MED_SilentTempo/Meshes/F_MED_SilentTempo_Physics
bodies {
 dyn_string_01_l [{
   shapeType=Capsule
   length=3.6965
   radius=0.5986
   shapeName=dyn_string_01_l_capsule
   center=<2.2398 0.2337 -0.2055>
   rotation=<rot 84.7586 0.0001 -13.3222>
  }]
 dyn_string_01_r [{shapeType=Capsule, length=3.9027, radius=0.644, shapeName=dyn_string_01_r_capsule, center=<2.064 0 0.0798>, rotation=<rot 83.2241 0 0>}]
 dyn_string_02_l [{shapeType=Capsule, length=3.4877, radius=0.6966, shapeName=dyn_string_02_l_capsule, center=<2.25 0 0>, rotation=<rot 90 0 0>}]
 dyn_string_02_r [{shapeType=Capsule, length=4.5165, radius=0.8118, shapeName=dyn_string_02_r_capsule, center=<1.9528 0 0>, rotation=<rot 90 0 0>}]
 dyn_string_03_l [{
   shapeType=Capsule
   length=4.3844
   radius=0.616
   shapeName=dyn_string_03_l_capsule
   center=<2.4179 0 0.1935>
   rotation=<rot 89.6022 -0.0003 -0.0003>
  }]
 dyn_string_03_r [{shapeType=Capsule, length=6.075, radius=0.5925, shapeName=dyn_string_03_r_capsule, center=<3.1494 0 0>, rotation=<rot 90 0 0>}]
 dyn_string_base_l [{shapeType=Sphere, radius=0.5, shapeName=dyn_string_base_l_sphere}]
 dyn_string_base_r [{shapeType=Sphere, radius=0.5, shapeName=dyn_string_base_r_sphere}]
 dyn_string_parent_l [{
   shapeType=Capsule
   length=1.6503
   radius=0.6546
   shapeName=dyn_string_parent_l_capsule
   center=<1.2683 -0.0398 0>
   rotation=<rot 90 180 167.6084>
  }]
 dyn_string_parent_r [{shapeType=Capsule, length=2.787, radius=0.5, shapeName=dyn_string_parent_r_capsule, center=<1.3442 0 0>, rotation=<rot 90 0 0>}]
 spine_04 [{shapeType=Capsule, length=12.6275, radius=5.3944, shapeName=spine_04_capsule, center=<-0.8406 7.9724 0>}]
 spine_05 [{shapeType=Box, extentX=19.7611, extentY=7.5, extentZ=25.3128, shapeName=spine_05_box, center=<-1.901 5.838 0>, rotation=<rot 0 -27.431 0>}]
}
constraints [
 {bone1Name=dyn_string_03_r, bone2Name=dyn_string_02_r, swing1LimitDegrees=20, swing2LimitDegrees=20, twistLimitDegrees=3}
 {bone1Name=dyn_string_parent_r, bone2Name=dyn_string_base_r, swing1LimitDegrees=20, swing2LimitDegrees=20, twistLimitDegrees=3}
 {bone1Name=dyn_string_01_r, bone2Name=dyn_string_parent_r, swing1LimitDegrees=20, swing2LimitDegrees=20, twistLimitDegrees=3}
 {bone1Name=dyn_string_02_r, bone2Name=dyn_string_01_r, swing1LimitDegrees=20, swing2LimitDegrees=20, twistLimitDegrees=3}
 {bone1Name=dyn_string_03_l, bone2Name=dyn_string_02_l, swing1LimitDegrees=20, swing2LimitDegrees=20, twistLimitDegrees=3}
 {bone1Name=dyn_string_parent_l, bone2Name=dyn_string_base_l, swing1LimitDegrees=30, swing2LimitDegrees=30, twistLimitDegrees=3}
 {bone1Name=dyn_string_01_l, bone2Name=dyn_string_parent_l, swing1LimitDegrees=30, swing2LimitDegrees=30, twistLimitDegrees=3}
 {bone1Name=dyn_string_02_l, bone2Name=dyn_string_01_l, swing1LimitDegrees=20, swing2LimitDegrees=20, twistLimitDegrees=3}
]
