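# Animation Blueprint for the SilentTempo female body cosmetic, adding secondary motion to a pose copied from the attached parent mesh.
# Weighted constraints place accessory bases, while AnimDynamics and three RigidBody simulations move zippers, sleeves, tassels and other dangling parts.
# Control Rigs adjust the chain from movement state and drive the clock from game time; an emote slot can override the hands.
# A cosmetic tag check enables an offset on the backpack attach bone.
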AnimBlueprint /BRCosmetics/Characters/Player/Female/Medium/Bodies/F_MED_SilentTempo/Meshes/F_MED_SilentTempo_AnimBP
target_skeleton=/Game/Characters/Player/Female/Medium/Bodies/F_MED_SilentTempo/Meshes/F_MED_SilentTempo_Skeleton
graphs [
 AnimGraph {nodes [
   "Constraint - dyn_belt_tassel_j_base (PP)"!(ComponentPose, Alpha=1, ConstraintWeights_0=0.7406, ConstraintWeights_1:float=0.2594) -> (Pose) {
    alphaBoolBlend {blendOption=Linear}
    boneToModify {boneName=dyn_belt_tassel_j_base}
    constraintSetup [{targetBone {boneName=dyn_belt_side_parent_l}}, {targetBone {boneName=dyn_belt_frnt_parent_l}}]
    constraintWeights [1,1]
   }
   n1 = Get HeadRotate_Alpha() -> (HeadRotate_Alpha:double)
   n2 = "Transform (Modify) Bone - Bone: dyn_collar_frnt_parent_r"(
    ComponentPose=Local To Component(LocalPose=Copy Pose From Mesh() {bUseAttachedParent=true})
    Alpha=@n1
    Translation="0,0,0"
    Rotation="-25,0,0"
    Scale="1,1,1"
   ) {
    rotationMode=BMM_Additive
    rotationSpace=BCS_BoneSpace
    translationMode=BMM_Additive
    translationSpace=BCS_BoneSpace
    alphaBoolBlend {blendOption=Linear}
    boneToModify {boneName=dyn_collar_frnt_parent_r}
    rotation {pitch=-25}
   }
   n3 = "Transform (Modify) Bone - Bone: dyn_collar_frnt_parent_l"(ComponentPose=@n2, Alpha=@n1, Translation="0,0,0", Rotation="-30,0,5", Scale="1,1,1") {
    rotationMode=BMM_Additive
    rotationSpace=BCS_BoneSpace
    translationMode=BMM_Additive
    translationSpace=BCS_BoneSpace
    alphaBoolBlend {blendOption=Linear}
    boneToModify {boneName=dyn_collar_frnt_parent_l}
    rotation {pitch=-30, roll=5}
   }
   n4 = "Constraint - dyn_zipper_neck_base (P)"!(ComponentPose=@n3, Alpha=1, ConstraintWeights_0=1, ConstraintWeights_1:float=0) -> (Pose) {
    alphaBoolBlend {blendOption=Linear}
    boneToModify {boneName=dyn_zipper_neck_base}
    constraintSetup [{targetBone {boneName=spine_05}}]
    constraintWeights [1,0]
   }
   n5 = "Constraint - dyn_zipper_stomach_base (PPPPP)"!(
    ComponentPose=@n4
    Alpha=1
    ConstraintWeights_0=0.2456
    ConstraintWeights_1:float=0.0581
    ConstraintWeights_2:float=0.0236
    ConstraintWeights_3:float=0.1433
    ConstraintWeights_4:float=0.5293
   ) -> (Pose) {
    alphaBoolBlend {blendOption=Linear}
    boneToModify {boneName=dyn_zipper_stomach_base}
    constraintSetup [
     {targetBone {boneName=spine_02}}
     {targetBone {boneName=spine_04}}
     {targetBone {boneName=spine_03}}
     {targetBone {boneName=spine_01}}
     {targetBone {boneName=pelvis}}
    ]
    constraintWeights [1,1,1,1,1]
   }
   n6 = "Constraint - dyn_zipper_leg_base_l (PPP)"!(ComponentPose=@n5, Alpha=1, ConstraintWeights_0=0.1241, ConstraintWeights_1:float=0.0191, ConstraintWeights_2:float=0.8568) -> (Pose) {
    alphaBoolBlend {blendOption=Linear}
    boneToModify {boneName=dyn_zipper_leg_base_l}
    constraintSetup [{targetBone {boneName=calf_l}}, {targetBone {boneName=calf_twist_01_l}}, {targetBone {boneName=calf_twist_02_l}}]
    constraintWeights [1,1,1]
   }
   n7 = Anim Dynamics - Bone: dyn_zipper_neck_parent(ComponentPose=@n6, Alpha=1, ExternalForce="0,0,0") {
    angularDampingOverride=1
    bGravityOverrideInSimSpace=true
    bOverrideAngularDamping=true
    bOverrideLinearDamping=true
    bUseGravityOverride=true
    linearDampingOverride=1
    alphaBoolBlend {blendOption=Linear}
    boundBone {boneName=dyn_zipper_neck_parent}
    componentAppliedLinearAccClamp=<3000 3000 3000>
    componentLinearAccScale=<0.15 0.15 0.15>
    componentLinearVelScale=<0.15 0.15 0.15>
    gravityOverride {z=-980}
    physicsBodyDefinitions [{
      sphereCollisionRadius=0
      boundBone {boneName=dyn_zipper_neck_parent}
      boxExtents=<6 4 2>
      constraintSetup {angularLimitsMax=<0 60 30>, angularLimitsMin=<0 0 -30>, angularTarget=<0 0 0>}
      localJointOffset=<2.5 0 0>
     }]
    relativeSpaceBone {boneName=root}
   }
   n8 = Anim Dynamics - Bone: dyn_zipper_stomach_parent(ComponentPose=@n7, Alpha=1, ExternalForce="0,0,0") {
    angularDampingOverride=1
    bGravityOverrideInSimSpace=true
    bOverrideAngularDamping=true
    bOverrideLinearDamping=true
    bUseGravityOverride=true
    linearDampingOverride=1
    alphaBoolBlend {blendOption=Linear}
    boundBone {boneName=dyn_zipper_stomach_parent}
    componentAppliedLinearAccClamp=<3000 3000 3000>
    componentLinearAccScale=<0.15 0.15 0.15>
    componentLinearVelScale=<0.15 0.15 0.15>
    gravityOverride {z=-980}
    physicsBodyDefinitions [{
      sphereCollisionRadius=0
      boundBone {boneName=dyn_zipper_stomach_parent}
      boxExtents=<6 4 2>
      constraintSetup {angularLimitsMax=<0 60 30>, angularLimitsMin=<0 -8 -30>, angularTarget=<0 0 0>}
      localJointOffset=<2.5 0 0>
     }]
    relativeSpaceBone {boneName=root}
   }
   n9 = Anim Dynamics - Bone: dyn_zipper_leg_parent_l(ComponentPose=@n8, Alpha=1, ExternalForce="0,0,0") {
    angularDampingOverride=1
    bGravityOverrideInSimSpace=true
    bOverrideAngularDamping=true
    bOverrideLinearDamping=true
    bUseGravityOverride=true
    linearDampingOverride=1
    alphaBoolBlend {blendOption=Linear}
    boundBone {boneName=dyn_zipper_leg_parent_l}
    componentAppliedLinearAccClamp=<3000 3000 3000>
    componentLinearAccScale=<0.15 0.15 0.15>
    componentLinearVelScale=<0.15 0.15 0.15>
    gravityOverride {z=-980}
    physicsBodyDefinitions [{
      sphereCollisionRadius=0
      boundBone {boneName=dyn_zipper_leg_parent_l}
      boxExtents=<6 4 2>
      constraintSetup {angularLimitsMax=<0 60 30>, angularLimitsMin=<0 -10 -30>, angularTarget=<0 0 0>}
      localJointOffset=<2.5 0 0>
     }]
    relativeSpaceBone {boneName=root}
   }
   n10 = Anim Dynamics - Bone: dyn_sleeve_inner_base_r(ComponentPose=@n9, Alpha=1, ExternalForce="0,0,0") {
    angularDampingOverride=0.99
    angularSpringConstant=1000
    bAngularSpring=true
    bGravityOverrideInSimSpace=true
    bLinearSpring=true
    bOverrideAngularDamping=true
    bUseGravityOverride=true
    bUsePlanarLimit=false
    linearSpringConstant=1000
    alphaBoolBlend {blendOption=Linear}
    boundBone {boneName=dyn_sleeve_inner_base_r}
    componentAppliedLinearAccClamp=<3000 3000 3000>
    componentLinearAccScale=<1 1 1>
    componentLinearVelScale=<1 1 1>
    gravityOverride {z=-980}
    physicsBodyDefinitions [{
      sphereCollisionRadius=0
      boundBone {boneName=dyn_sleeve_inner_base_r}
      boxExtents=<20 20 10>
      constraintSetup {angularLimitsMax=<0 5 6>, angularLimitsMin=<0 -5 -8>, angularTarget=<1 0 0>}
      localJointOffset=<6.5 0 0>
     }]
    relativeSpaceBone {boneName=root}
   }
   n11 = Anim Dynamics - Bone: dyn_sleeve_outter_base_r(ComponentPose=@n10, Alpha=1, ExternalForce="0,0,0") {
    angularDampingOverride=0.99
    angularSpringConstant=1000
    bAngularSpring=true
    bGravityOverrideInSimSpace=true
    bLinearSpring=true
    bOverrideAngularDamping=true
    bUseGravityOverride=true
    bUsePlanarLimit=false
    linearSpringConstant=1000
    alphaBoolBlend {blendOption=Linear}
    boundBone {boneName=dyn_sleeve_outter_base_r}
    componentAppliedLinearAccClamp=<3000 3000 3000>
    componentLinearAccScale=<1 1 1>
    componentLinearVelScale=<1 1 1>
    gravityOverride {z=-980}
    physicsBodyDefinitions [{
      sphereCollisionRadius=0
      boundBone {boneName=dyn_sleeve_outter_base_r}
      boxExtents=<20 20 10>
      constraintSetup {angularLimitsMax=<0 10 10>, angularLimitsMin=<0 -5 0>, angularTarget=<1 0 0>}
      localJointOffset=<6.5 0 0>
     }]
    relativeSpaceBone {boneName=root}
   }
   n12 = Anim Dynamics - Bone: dyn_sleeve_bottom_base_r(ComponentPose=@n11, Alpha=1, ExternalForce="0,0,0") {
    angularDampingOverride=0.99
    angularSpringConstant=1000
    bGravityOverrideInSimSpace=true
    bOverrideAngularDamping=true
    bUseGravityOverride=true
    bUsePlanarLimit=false
    linearSpringConstant=1000
    alphaBoolBlend {blendOption=Linear}
    boundBone {boneName=dyn_sleeve_bottom_base_r}
    componentAppliedLinearAccClamp=<3000 3000 3000>
    componentLinearAccScale=<1 1 1>
    componentLinearVelScale=<1 1 1>
    gravityOverride {z=-980}
    physicsBodyDefinitions [{
      sphereCollisionRadius=0
      boundBone {boneName=dyn_sleeve_bottom_base_r}
      boxExtents=<20 20 10>
      constraintSetup {angularLimitsMax=<0 10 5>, angularLimitsMin=<0 0 -5>, angularTarget=<1 0 0>}
      localJointOffset=<6.5 0 0>
     }]
    relativeSpaceBone {boneName=root}
   }
   n13 = Anim Dynamics - Bone: dyn_sleeve_top_base_r(ComponentPose=@n12, Alpha=1, ExternalForce="0,0,0") {
    angularDampingOverride=0.99
    angularSpringConstant=1000
    bAngularSpring=true
    bGravityOverrideInSimSpace=true
    bLinearSpring=true
    bOverrideAngularDamping=true
    bUseGravityOverride=true
    bUsePlanarLimit=false
    linearSpringConstant=1000
    alphaBoolBlend {blendOption=Linear}
    boundBone {boneName=dyn_sleeve_top_base_r}
    componentAppliedLinearAccClamp=<3000 3000 3000>
    componentLinearAccScale=<1 1 1>
    componentLinearVelScale=<1 1 1>
    gravityOverride {z=-980}
    physicsBodyDefinitions [{
      sphereCollisionRadius=0
      boundBone {boneName=dyn_sleeve_top_base_r}
      boxExtents=<20 20 10>
      constraintSetup {angularLimitsMax=<0 5 5>, angularLimitsMin=<0 0 -5>, angularTarget=<1 0 0>}
      localJointOffset=<6.5 0 0>
     }]
    relativeSpaceBone {boneName=root}
   }
   n14 = "Constraint - dyn_string_base_l (PP)"!(ComponentPose=@n13, Alpha=1, ConstraintWeights_0=0.8794, ConstraintWeights_1:float=0.1206) -> (Pose) {
    alphaBoolBlend {blendOption=Linear}
    boneToModify {boneName=dyn_string_base_l}
    constraintSetup [{targetBone {boneName=spine_05}}, {targetBone {boneName=clavicle_l}}]
    constraintWeights [1,1]
   }
   n15 = "Constraint - dyn_string_base_r (PP)"!(ComponentPose=@n14, Alpha=1, ConstraintWeights_0=0.8787, ConstraintWeights_1:float=0.1213) -> (Pose) {
    alphaBoolBlend {blendOption=Linear}
    boneToModify {boneName=dyn_string_base_r}
    constraintSetup [{targetBone {boneName=spine_05}}, {targetBone {boneName=clavicle_r}}]
    constraintWeights [1,1]
   }
   n16 = RigidBody(ComponentPose=@n15) {
    bClampLinearTranslationLimitToRefPose=true
    bOverrideWorldGravity=true
    overridePhysicsAsset=&/F_MED_SilentTempo_Physics
    simulationSpace=BaseBoneSpace
    alphaBoolBlend {blendOption=Linear}
    baseBoneRef {boneName=spine_05}
    componentAppliedLinearAccClamp=<6000 6000 6000>
    componentLinearAccScale=<0.1 0.1 0.1>
    componentLinearVelScale=<0.1 0.1 0.1>
    overrideWorldGravity {z=-980}
    simSpaceSettings {
     maxAngularAcceleration=500
     maxAngularVelocity=3
     maxLinearAcceleration=400
     maxLinearVelocity=250
     velocityScaleZ=0
     worldAlpha=1
     externalLinearDragV=<0.1 0.1>
    }
   }
   n17 = "Constraint - dyn_belt_frnt_base_mid (PP)"!(ComponentPose=@n16, Alpha=1, ConstraintWeights_0=0.9816, ConstraintWeights_1:float=0.0184) -> (Pose) {
    alphaBoolBlend {blendOption=Linear}
    boneToModify {boneName=dyn_belt_frnt_base_mid}
    constraintSetup [{targetBone {boneName=pelvis}}, {targetBone {boneName=dfrm_thigh_l}}]
    constraintWeights [1,1]
   }
   n18 = "Constraint - dyn_belt_frnt_base_l (PP)"!(ComponentPose=@n17, Alpha=1, ConstraintWeights_0=0.8423, ConstraintWeights_1:float=0.1577) -> (Pose) {
    alphaBoolBlend {blendOption=Linear}
    boneToModify {boneName=dyn_belt_frnt_base_l}
    constraintSetup [{targetBone {boneName=pelvis}}, {targetBone {boneName=dfrm_thigh_l}}]
    constraintWeights [1,1]
   }
   n19 = "Constraint - dyn_belt_side_base_l (PP)"!(ComponentPose=@n18, Alpha=1, ConstraintWeights_0=0.6219, ConstraintWeights_1:float=0.3781) -> (Pose) {
    alphaBoolBlend {blendOption=Linear}
    boneToModify {boneName=dyn_belt_side_base_l}
    constraintSetup [{targetBone {boneName=pelvis}}, {targetBone {boneName=dfrm_thigh_l}}]
    constraintWeights [1,1]
   }
   n20 = "Constraint - dyn_belt_bk_base_l (PP)"!(ComponentPose=@n19, Alpha=1, ConstraintWeights_0=0.9759, ConstraintWeights_1:float=0.0241) -> (Pose) {
    alphaBoolBlend {blendOption=Linear}
    boneToModify {boneName=dyn_belt_bk_base_l}
    constraintSetup [{targetBone {boneName=pelvis}}, {targetBone {boneName=dfrm_thigh_l}}]
    constraintWeights [1,1]
   }
   n21 = Get gravity_override() -> (gravity_override:Vector)
   n22 = RigidBody!(ComponentPose=@n20, Alpha, OverrideWorldGravity:Vector=@n21) -> (Pose) {
    bClampLinearTranslationLimitToRefPose=true
    bOverrideWorldGravity=true
    overridePhysicsAsset=&/F_MED_SilentTempo_Chain_Physics
    simulationSpace=BaseBoneSpace
    alphaBoolBlend {blendOption=Linear}
    baseBoneRef {boneName=pelvis}
    componentAppliedLinearAccClamp=<2500 2500 2500>
    componentLinearAccScale=<0.4 0.3 0.1>
    componentLinearVelScale=<0.4 0.3 0.1>
    overrideWorldGravity {z=-980}
    simSpaceSettings {maxAngularAcceleration=1000, maxAngularVelocity=3, maxLinearAcceleration=250, maxLinearVelocity=500, velocityScaleZ=0, worldAlpha=1}
   }
   n23 = Control Rig!(
    Source=Component To Local(ComponentPose=@n22)
    Alpha
    spine_twist_right_delta_ref_alpha:double=Get spine_twist_right_delta_ref_alpha() -> (spine_twist_right_delta_ref_alpha:double)
    left_right_movement_alpha:double=Get left_right_movement_alpha() -> (left_right_movement_alpha:double)
    pelvis_tilt_left_alpha:double=Get pelvis_tilt_left_alpha() -> (pelvis_tilt_left_alpha:double)
    is_crouching:bool=Get bIsCrouching()
    is_tactical_sprinting:bool=Get bIsTacticalSprinting()
    is_frontend:bool=Get bIsInFrontEnd()
   ) -> (Pose) {
    alphaBoolBlend {blendOption=Linear}
    controlRigAssetReference {blueprintRigClass=&/F_MED_SilentTempo_Chain_CtrlRig.F_MED_SilentTempo_Chain_CtrlRig_C}
   }
   n24 = "Constraint - dyn_clock_chain_base (PPP)"!(
    ComponentPose=Local To Component(LocalPose=@n23)
    Alpha=1
    ConstraintWeights_0=0.7738
    ConstraintWeights_1:float=0.1313
    ConstraintWeights_2:float=0.0949
   ) -> (Pose) {
    alphaBoolBlend {blendOption=Linear}
    boneToModify {boneName=dyn_clock_chain_base}
    constraintSetup [{targetBone {boneName=pelvis}}, {targetBone {boneName=spine_02}}, {targetBone {boneName=spine_01}}]
    constraintWeights [1,1,1]
   }
   n25 = RigidBody(ComponentPose=@n24, OverrideWorldGravity=@n21) {
    bClampLinearTranslationLimitToRefPose=true
    bOverrideWorldGravity=true
    overridePhysicsAsset=&/F_MED_SilentTempo_Clock_Physics
    simulationSpace=BaseBoneSpace
    alphaBoolBlend {blendOption=Linear}
    baseBoneRef {boneName=pelvis}
    componentAppliedLinearAccClamp=<2500 2500 2500>
    componentLinearAccScale=<0.4 0.3 0.1>
    componentLinearVelScale=<0.4 0.3 0.1>
    overrideWorldGravity {z=-980}
    simSpaceSettings {maxAngularAcceleration=1000, maxAngularVelocity=3, maxLinearAcceleration=250, maxLinearVelocity=500, velocityScaleZ=0, worldAlpha=1}
   }
   n26 = "Constraint - dyn_belt_tassel_a_base (PPP)"!(ComponentPose=@n25, Alpha=1, ConstraintWeights_0=0.727, ConstraintWeights_1:float=0.1471, ConstraintWeights_2:float=0.1259) -> (Pose) {
    alphaBoolBlend {blendOption=Linear}
    boneToModify {boneName=dyn_belt_tassel_a_base}
    constraintSetup [{targetBone {boneName=pelvis}}, {targetBone {boneName=spine_02}}, {targetBone {boneName=spine_01}}]
    constraintWeights [1,1,1]
   }
   n27 = "Constraint - dyn_belt_tassel_b_base (PPPP)"!(
    ComponentPose=@n26
    Alpha=1
    ConstraintWeights_0=0.7668
    ConstraintWeights_1:float=0.1118
    ConstraintWeights_2:float=0.1062
    ConstraintWeights_3:float=0.0152
   ) -> (Pose) {
    alphaBoolBlend {blendOption=Linear}
    boneToModify {boneName=dyn_belt_tassel_b_base}
    constraintSetup [
     {targetBone {boneName=pelvis}}
     {targetBone {boneName=spine_01}}
     {targetBone {boneName=spine_02}}
     {targetBone {boneName=dyn_belt_bk_parent_l}}
    ]
    constraintWeights [1,1,1,1]
   }
   n28 = "Constraint - dyn_belt_tassel_c_base (PPPP)"!(
    ComponentPose=@n27
    Alpha=1
    ConstraintWeights_0=0.7627
    ConstraintWeights_1:float=0.1023
    ConstraintWeights_2:float=0.0723
    ConstraintWeights_3:float=0.0626
   ) -> (Pose) {
    alphaBoolBlend {blendOption=Linear}
    boneToModify {boneName=dyn_belt_tassel_c_base}
    constraintSetup [
     {targetBone {boneName=pelvis}}
     {targetBone {boneName=spine_01}}
     {targetBone {boneName=spine_02}}
     {targetBone {boneName=dyn_belt_bk_parent_l}}
    ]
    constraintWeights [1,1,1,1]
   }
   n29 = "Constraint - dyn_belt_tassel_d_base (PPPP)"!(
    ComponentPose=@n28
    Alpha=1
    ConstraintWeights_0=0.5641
    ConstraintWeights_1:float=0.2316
    ConstraintWeights_2:float=0.152
    ConstraintWeights_3:float=0.0523
   ) -> (Pose) {
    alphaBoolBlend {blendOption=Linear}
    boneToModify {boneName=dyn_belt_tassel_d_base}
    constraintSetup [
     {targetBone {boneName=pelvis}}
     {targetBone {boneName=dyn_belt_bk_parent_l}}
     {targetBone {boneName=spine_01}}
     {targetBone {boneName=spine_02}}
    ]
    constraintWeights [1,1,1,1]
   }
   n30 = "Constraint - dyn_belt_tassel_e_base (PPPPP)"!(
    ComponentPose=@n29
    Alpha=1
    ConstraintWeights_0=0.4426
    ConstraintWeights_1:float=0.2545
    ConstraintWeights_2:float=0.184
    ConstraintWeights_3:float=0.0923
    ConstraintWeights_4:float=0.0265
   ) -> (Pose) {
    alphaBoolBlend {blendOption=Linear}
    boneToModify {boneName=dyn_belt_tassel_e_base}
    constraintSetup [
     {targetBone {boneName=dyn_belt_bk_parent_l}}
     {targetBone {boneName=pelvis}}
     {targetBone {boneName=spine_01}}
     {targetBone {boneName=dyn_belt_side_parent_l}}
     {targetBone {boneName=spine_02}}
    ]
    constraintWeights [1,1,1,1,1]
   }
   n31 = "Constraint - dyn_belt_tassel_f_base (PPPP)"!(
    ComponentPose=@n30
    Alpha=1
    ConstraintWeights_0=0.5923
    ConstraintWeights_1:float=0.3038
    ConstraintWeights_2:float=0.0735
    ConstraintWeights_3:float=0.0304
   ) -> (Pose) {
    alphaBoolBlend {blendOption=Linear}
    boneToModify {boneName=dyn_belt_tassel_f_base}
    constraintSetup [
     {targetBone {boneName=dyn_belt_side_parent_l}}
     {targetBone {boneName=dyn_belt_bk_parent_l}}
     {targetBone {boneName=spine_01}}
     {targetBone {boneName=pelvis}}
    ]
    constraintWeights [1,1,1,1]
   }
   n32 = "Constraint - dyn_belt_tassel_g_base (PP)"!(ComponentPose=@n31, Alpha=1, ConstraintWeights_0=0.8738, ConstraintWeights_1:float=0.1262) -> (Pose) {
    alphaBoolBlend {blendOption=Linear}
    boneToModify {boneName=dyn_belt_tassel_g_base}
    constraintSetup [{targetBone {boneName=dyn_belt_side_parent_l}}, {targetBone {boneName=dyn_belt_bk_parent_l}}]
    constraintWeights [1,1]
   }
   n33 = "Constraint - dyn_belt_tassel_h_base (PP)"!(ComponentPose=@n32, Alpha=1, ConstraintWeights_0=0.99, ConstraintWeights_1:float=0.01) -> (Pose) {
    alphaBoolBlend {blendOption=Linear}
    boneToModify {boneName=dyn_belt_tassel_h_base}
    constraintSetup [{targetBone {boneName=dyn_belt_side_parent_l}}, {targetBone {boneName=dyn_belt_bk_parent_l}}]
    constraintWeights [1,1]
   }
   n34 = "Constraint - dyn_belt_tassel_i_base (PP)"!(ComponentPose=@n33, Alpha=1, ConstraintWeights_0=0.9768, ConstraintWeights_1:float=0.0232) -> (Pose) {
    alphaBoolBlend {blendOption=Linear}
    boneToModify {boneName=dyn_belt_tassel_i_base}
    constraintSetup [{targetBone {boneName=dyn_belt_side_parent_l}}, {targetBone {boneName=dyn_belt_frnt_parent_l}}]
    constraintWeights [1,1]
   }
   n35 = "Constraint - dyn_belt_tassel_k_base (PPP)"!(ComponentPose=@n34, Alpha=1, ConstraintWeights_0=0.8741, ConstraintWeights_1:float=0.0752, ConstraintWeights_2:float=0.0507) -> (Pose) {
    alphaBoolBlend {blendOption=Linear}
    boneToModify {boneName=dyn_belt_tassel_k_base}
    constraintSetup [{targetBone {boneName=pelvis}}, {targetBone {boneName=spine_02}}, {targetBone {boneName=spine_01}}]
    constraintWeights [1,1,1]
   }
   n36 = "Constraint - dyn_belt_tassel_l_base (PP)"!(ComponentPose=@n35, Alpha=1, ConstraintWeights_0=0.9732, ConstraintWeights_1:float=0.0268) -> (Pose) {
    alphaBoolBlend {blendOption=Linear}
    boneToModify {boneName=dyn_belt_tassel_l_base}
    constraintSetup [{targetBone {boneName=pelvis}}, {targetBone {boneName=dyn_belt_frnt_parent_mid}}]
    constraintWeights [1,1]
   }
   n37 = "Constraint - dyn_belt_tassel_m_base (PP)"!(ComponentPose=@n36, Alpha=1, ConstraintWeights_0=0.6696, ConstraintWeights_1:float=0.3304) -> (Pose) {
    alphaBoolBlend {blendOption=Linear}
    boneToModify {boneName=dyn_belt_tassel_m_base}
    constraintSetup [{targetBone {boneName=pelvis}}, {targetBone {boneName=dyn_belt_frnt_parent_mid}}]
    constraintWeights [1,1]
   }
   n38 = "Constraint - dyn_belt_tassel_n_base (PPP)"!(ComponentPose=@n37, Alpha=1, ConstraintWeights_0=0.8244, ConstraintWeights_1:float=0.1629, ConstraintWeights_2:float=0.0127) -> (Pose) {
    alphaBoolBlend {blendOption=Linear}
    boneToModify {boneName=dyn_belt_tassel_n_base}
    constraintSetup [{targetBone {boneName=dyn_belt_frnt_parent_mid}}, {targetBone {boneName=pelvis}}, {targetBone {boneName=dyn_belt_frnt_parent_l}}]
    constraintWeights [1,1,1]
   }
   n39 = "Constraint - dyn_belt_tassel_o_base (PPP)"!(ComponentPose=@n38, Alpha=1, ConstraintWeights_0=0.7254, ConstraintWeights_1:float=0.2098, ConstraintWeights_2:float=0.0648) -> (Pose) {
    alphaBoolBlend {blendOption=Linear}
    boneToModify {boneName=dyn_belt_tassel_o_base}
    constraintSetup [{targetBone {boneName=dyn_belt_frnt_parent_mid}}, {targetBone {boneName=dyn_belt_frnt_parent_l}}, {targetBone {boneName=pelvis}}]
    constraintWeights [1,1,1]
   }
   n40 = "Constraint - dyn_belt_tassel_p_base (PPP)"!(ComponentPose=@n39, Alpha=1, ConstraintWeights_0=0.6038, ConstraintWeights_1:float=0.3555, ConstraintWeights_2:float=0.0407) -> (Pose) {
    alphaBoolBlend {blendOption=Linear}
    boneToModify {boneName=dyn_belt_tassel_p_base}
    constraintSetup [
     {targetBone {boneName=dyn_belt_frnt_parent_l}}
     {targetBone {boneName=dyn_belt_frnt_parent_mid}}
     {targetBone {boneName=dyn_belt_side_parent_l}}
    ]
    constraintWeights [1,1,1]
   }
   n41 = Anim Dynamics - Bone: dyn_belt_tassel_a_parent(ComponentPose=@n40, Alpha=1, ExternalForce="0,0,0") {
    angularDampingOverride=0.9999
    angularSpringConstant=50
    bAngularSpring=true
    bGravityOverrideInSimSpace=true
    bOverrideAngularDamping=true
    bOverrideLinearDamping=true
    bUseGravityOverride=true
    linearDampingOverride=0.9999
    simulationSpace=RootRelative
    alphaBoolBlend {blendOption=Linear}
    boundBone {boneName=dyn_belt_tassel_a_parent}
    componentAppliedLinearAccClamp=<6000 6000 6000>
    componentLinearAccScale=<0.1 0.1 0.1>
    componentLinearVelScale=<0.1 0.1 0.1>
    gravityOverride {z=-980}
    physicsBodyDefinitions [{
      sphereCollisionRadius=0
      boundBone {boneName=dyn_belt_tassel_a_parent}
      boxExtents=<12 12 4>
      constraintSetup {angularLimitsMax=<0 40 30>, angularLimitsMin=<0 10 -30>, angularTarget=<1 0 0>}
      localJointOffset=<3.5 0 0>
     }]
    relativeSpaceBone {boneName=pelvis}
   }
   n42 = Anim Dynamics - Bone: dyn_belt_tassel_b_parent(ComponentPose=@n41, Alpha=1, ExternalForce="0,0,0") {
    angularDampingOverride=0.9999
    angularSpringConstant=50
    bAngularSpring=true
    bGravityOverrideInSimSpace=true
    bOverrideAngularDamping=true
    bOverrideLinearDamping=true
    bUseGravityOverride=true
    linearDampingOverride=0.9999
    simulationSpace=RootRelative
    alphaBoolBlend {blendOption=Linear}
    boundBone {boneName=dyn_belt_tassel_b_parent}
    componentAppliedLinearAccClamp=<6000 6000 6000>
    componentLinearAccScale=<0.1 0.1 0.1>
    componentLinearVelScale=<0.1 0.1 0.1>
    gravityOverride {z=-980}
    physicsBodyDefinitions [{
      sphereCollisionRadius=0
      boundBone {boneName=dyn_belt_tassel_b_parent}
      boxExtents=<12 12 4>
      constraintSetup {angularLimitsMax=<0 50 30>, angularLimitsMin=<0 10 -30>, angularTarget=<1 0 0>}
      localJointOffset=<3.5 0 0>
     }]
    relativeSpaceBone {boneName=root}
   }
   n43 = Anim Dynamics - Bone: dyn_belt_tassel_c_parent(ComponentPose=@n42, Alpha=1, ExternalForce="0,0,0") {
    angularDampingOverride=0.9999
    angularSpringConstant=50
    bAngularSpring=true
    bGravityOverrideInSimSpace=true
    bOverrideAngularDamping=true
    bOverrideLinearDamping=true
    bUseGravityOverride=true
    linearDampingOverride=0.9999
    simulationSpace=RootRelative
    alphaBoolBlend {blendOption=Linear}
    boundBone {boneName=dyn_belt_tassel_c_parent}
    componentAppliedLinearAccClamp=<6000 6000 6000>
    componentLinearAccScale=<0.1 0.1 0.1>
    componentLinearVelScale=<0.1 0.1 0.1>
    gravityOverride {z=-980}
    physicsBodyDefinitions [{
      sphereCollisionRadius=0
      boundBone {boneName=dyn_belt_tassel_c_parent}
      boxExtents=<12 12 4>
      constraintSetup {angularLimitsMax=<0 60 30>, angularLimitsMin=<0 10 -30>, angularTarget=<1 0 0>}
      localJointOffset=<3.5 0 0>
     }]
    relativeSpaceBone {boneName=root}
   }
   n44 = Anim Dynamics - Bone: dyn_belt_tassel_f_parent(ComponentPose=@n43, Alpha=1, ExternalForce="0,0,0") {
    angularDampingOverride=0.9999
    angularSpringConstant=50
    bAngularSpring=true
    bGravityOverrideInSimSpace=true
    bOverrideAngularDamping=true
    bOverrideLinearDamping=true
    bUseGravityOverride=true
    linearDampingOverride=0.9999
    simulationSpace=RootRelative
    alphaBoolBlend {blendOption=Linear}
    boundBone {boneName=dyn_belt_tassel_f_parent}
    componentAppliedLinearAccClamp=<6000 6000 6000>
    componentLinearAccScale=<0.1 0.1 0.1>
    componentLinearVelScale=<0.1 0.1 0.1>
    gravityOverride {z=-980}
    physicsBodyDefinitions [{
      sphereCollisionRadius=0
      boundBone {boneName=dyn_belt_tassel_f_parent}
      boxExtents=<12 12 4>
      constraintSetup {angularLimitsMax=<0 60 30>, angularLimitsMin=<0 0 -30>, angularTarget=<1 0 0>}
      localJointOffset=<3.5 0 0>
     }]
    relativeSpaceBone {boneName=root}
   }
   n45 = Anim Dynamics - Bone: dyn_belt_tassel_e_parent(ComponentPose=@n44, Alpha=1, ExternalForce="0,0,0") {
    angularDampingOverride=0.9999
    angularSpringConstant=50
    bAngularSpring=true
    bGravityOverrideInSimSpace=true
    bOverrideAngularDamping=true
    bOverrideLinearDamping=true
    bUseGravityOverride=true
    linearDampingOverride=0.9999
    simulationSpace=RootRelative
    alphaBoolBlend {blendOption=Linear}
    boundBone {boneName=dyn_belt_tassel_e_parent}
    componentAppliedLinearAccClamp=<6000 6000 6000>
    componentLinearAccScale=<0.1 0.1 0.1>
    componentLinearVelScale=<0.1 0.1 0.1>
    gravityOverride {z=-980}
    physicsBodyDefinitions [{
      sphereCollisionRadius=0
      boundBone {boneName=dyn_belt_tassel_e_parent}
      boxExtents=<12 7 7>
      constraintSetup {angularLimitsMax=<0 60 30>, angularLimitsMin=<0 1 -30>, angularTarget=<1 0 0>}
      localJointOffset=<4 0 0>
     }]
    relativeSpaceBone {boneName=root}
   }
   n46 = Anim Dynamics - Bone: dyn_belt_tassel_d_parent(ComponentPose=@n45, Alpha=1, ExternalForce="0,0,0") {
    angularDampingOverride=0.9999
    angularSpringConstant=50
    bAngularSpring=true
    bGravityOverrideInSimSpace=true
    bOverrideAngularDamping=true
    bOverrideLinearDamping=true
    linearDampingOverride=0.9999
    simulationSpace=RootRelative
    alphaBoolBlend {blendOption=Linear}
    boundBone {boneName=dyn_belt_tassel_d_parent}
    componentAppliedLinearAccClamp=<6000 6000 6000>
    componentLinearAccScale=<0.1 0.1 0.1>
    componentLinearVelScale=<0.1 0.1 0.1>
    gravityOverride {z=-980}
    physicsBodyDefinitions [{
      sphereCollisionRadius=0
      boundBone {boneName=dyn_belt_tassel_d_parent}
      boxExtents=<12 12 4>
      constraintSetup {angularLimitsMax=<0 60 40>, angularLimitsMin=<0 10 -40>, angularTarget=<1 0 0>}
      localJointOffset=<3.5 0 0>
     }]
    relativeSpaceBone {boneName=root}
   }
   n47 = Anim Dynamics - Bone: dyn_belt_tassel_g_parent(ComponentPose=@n46, Alpha=1, ExternalForce="0,0,0") {
    angularDampingOverride=0.9999
    angularSpringConstant=50
    bAngularSpring=true
    bGravityOverrideInSimSpace=true
    bOverrideAngularDamping=true
    bOverrideLinearDamping=true
    bUseGravityOverride=true
    linearDampingOverride=0.9999
    simulationSpace=RootRelative
    alphaBoolBlend {blendOption=Linear}
    boundBone {boneName=dyn_belt_tassel_g_parent}
    componentAppliedLinearAccClamp=<6000 6000 6000>
    componentLinearAccScale=<0.1 0.1 0.1>
    componentLinearVelScale=<0.1 0.1 0.1>
    gravityOverride {z=-980}
    physicsBodyDefinitions [{
      sphereCollisionRadius=0
      boundBone {boneName=dyn_belt_tassel_g_parent}
      boxExtents=<12 12 4>
      constraintSetup {angularLimitsMax=<0 60 30>, angularLimitsMin=<0 0 -30>, angularTarget=<1 0 0>}
      localJointOffset=<3.5 0 0>
     }]
    relativeSpaceBone {boneName=root}
   }
   n48 = Anim Dynamics - Bone: dyn_belt_tassel_h_parent(ComponentPose=@n47, Alpha=1, ExternalForce="0,0,0") {
    angularDampingOverride=0.9999
    angularSpringConstant=50
    bAngularSpring=true
    bGravityOverrideInSimSpace=true
    bOverrideAngularDamping=true
    bOverrideLinearDamping=true
    bUseGravityOverride=true
    bUseSphericalLimits=true
    linearDampingOverride=0.9999
    simulationSpace=RootRelative
    alphaBoolBlend {blendOption=Linear}
    boundBone {boneName=dyn_belt_tassel_h_parent}
    componentAppliedLinearAccClamp=<6000 6000 6000>
    componentLinearAccScale=<0.1 0.1 0.1>
    componentLinearVelScale=<0.1 0.1 0.1>
    gravityOverride {z=-980}
    physicsBodyDefinitions [{
      sphereCollisionRadius=0
      boundBone {boneName=dyn_belt_tassel_h_parent}
      boxExtents=<12 12 4>
      constraintSetup {angularLimitsMax=<0 60 30>, angularLimitsMin=<0 10 -30>, angularTarget=<1 0 0>}
      localJointOffset=<3.5 0 0>
     }]
    relativeSpaceBone {boneName=root}
    sphericalLimits [{limitRadius=10, drivingBone {boneName=pelvis}, sphereLocalOffset=<-6 0 -7>}]
   }
   n49 = Anim Dynamics - Bone: dyn_belt_tassel_i_parent(ComponentPose=@n48, Alpha=1, ExternalForce="0,0,0") {
    angularDampingOverride=0.9999
    angularSpringConstant=50
    bAngularSpring=true
    bGravityOverrideInSimSpace=true
    bOverrideAngularDamping=true
    bOverrideLinearDamping=true
    bUseGravityOverride=true
    bUseSphericalLimits=true
    linearDampingOverride=0.9999
    simulationSpace=RootRelative
    alphaBoolBlend {blendOption=Linear}
    boundBone {boneName=dyn_belt_tassel_i_parent}
    componentAppliedLinearAccClamp=<6000 6000 6000>
    componentLinearAccScale=<0.1 0.1 0.1>
    componentLinearVelScale=<0.1 0.1 0.1>
    gravityOverride {z=-980}
    physicsBodyDefinitions [{
      sphereCollisionRadius=0
      boundBone {boneName=dyn_belt_tassel_i_parent}
      boxExtents=<12 12 4>
      constraintSetup {angularLimitsMax=<0 40 30>, angularLimitsMin=<0 10 -20>, angularTarget=<1 0 0>}
      localJointOffset=<4 0 0>
     }]
    relativeSpaceBone {boneName=root}
    sphericalLimits [
     {limitRadius=17, drivingBone {boneName=pelvis}, sphereLocalOffset=<-2 0 1>}
     {limitRadius=8, drivingBone {boneName=thigh_l}, sphereLocalOffset=<-7 -1 0>}
    ]
   }
   n50 = Anim Dynamics - Bone: dyn_belt_tassel_l_parent(ComponentPose=@n49, Alpha=1, ExternalForce="0,0,0") {
    angularDampingOverride=0.9999
    angularSpringConstant=50
    bAngularSpring=true
    bGravityOverrideInSimSpace=true
    bOverrideAngularDamping=true
    bOverrideLinearDamping=true
    bUseGravityOverride=true
    bUseSphericalLimits=true
    linearDampingOverride=0.9999
    simulationSpace=RootRelative
    alphaBoolBlend {blendOption=Linear}
    boundBone {boneName=dyn_belt_tassel_l_parent}
    componentAppliedLinearAccClamp=<6000 6000 6000>
    componentLinearAccScale=<0.1 0.1 0.1>
    componentLinearVelScale=<0.1 0.1 0.1>
    gravityOverride {z=-980}
    physicsBodyDefinitions [{
      sphereCollisionRadius=0
      boundBone {boneName=dyn_belt_tassel_l_parent}
      boxExtents=<12 12 4>
      constraintSetup {angularLimitsMax=<0 60 30>, angularLimitsMin=<0 5 -30>, angularTarget=<1 0 0>}
      localJointOffset=<3.5 0 0>
     }]
    relativeSpaceBone {boneName=root}
    sphericalLimits [{limitRadius=14, drivingBone {boneName=thigh_r}, sphereLocalOffset=<0 5 -5>}]
   }
   n51 = Anim Dynamics - Bone: dyn_belt_tassel_k_parent(ComponentPose=@n50, Alpha=1, ExternalForce="0,0,0") {
    angularDampingOverride=0.9999
    angularSpringConstant=50
    bAngularSpring=true
    bGravityOverrideInSimSpace=true
    bOverrideAngularDamping=true
    bOverrideLinearDamping=true
    bUseGravityOverride=true
    bUseSphericalLimits=true
    linearDampingOverride=0.9999
    simulationSpace=RootRelative
    alphaBoolBlend {blendOption=Linear}
    boundBone {boneName=dyn_belt_tassel_k_parent}
    componentAppliedLinearAccClamp=<3000 3000 3000>
    componentLinearAccScale=<0.15 0.15 0.15>
    componentLinearVelScale=<0.15 0.15 0.15>
    gravityOverride {z=-980}
    physicsBodyDefinitions [{
      sphereCollisionRadius=0
      boundBone {boneName=dyn_belt_tassel_k_parent}
      boxExtents=<12 12 4>
      constraintSetup {angularLimitsMax=<0 60 30>, angularLimitsMin=<0 10 -30>, angularTarget=<1 0 0>}
      localJointOffset=<3.5 0 0>
     }]
    relativeSpaceBone {boneName=root}
    sphericalLimits [{limitRadius=10, drivingBone {boneName=pelvis}, sphereLocalOffset=<0 0 0>}]
   }
   n52 = Anim Dynamics - Bone: dyn_belt_tassel_j_parent(ComponentPose=@n51, Alpha=1, ExternalForce="0,0,0") {
    angularDampingOverride=1
    angularSpringConstant=50
    bAngularSpring=true
    bGravityOverrideInSimSpace=true
    bOverrideAngularDamping=true
    linearDampingOverride=0.999
    simulationSpace=RootRelative
    alphaBoolBlend {blendOption=Linear}
    boundBone {boneName=dyn_belt_tassel_j_parent}
    componentAppliedLinearAccClamp=<3000 3000 3000>
    componentLinearAccScale=<0.25 0.25 0.25>
    componentLinearVelScale=<0.25 0.25 0.25>
    physicsBodyDefinitions [{
      sphereCollisionRadius=0
      boundBone {boneName=dyn_belt_tassel_j_parent}
      boxExtents=<12 12 4>
      constraintSetup {angularLimitsMax=<0 40 15>, angularLimitsMin=<0 15 -15>, angularTarget=<1 0 0>}
      localJointOffset=<3.5 0 0>
     }]
    relativeSpaceBone {boneName=root}
   }
   n53 = Anim Dynamics - Bone: dyn_belt_tassel_m_parent(ComponentPose=@n52, Alpha=1, ExternalForce="0,0,0") {
    angularDampingOverride=0.9999
    angularSpringConstant=50
    bAngularSpring=true
    bGravityOverrideInSimSpace=true
    bOverrideAngularDamping=true
    bOverrideLinearDamping=true
    bUseSphericalLimits=true
    linearDampingOverride=0.9999
    simulationSpace=RootRelative
    alphaBoolBlend {blendOption=Linear}
    boundBone {boneName=dyn_belt_tassel_m_parent}
    componentAppliedLinearAccClamp=<3000 3000 3000>
    componentLinearAccScale=<0.15 0.15 0.15>
    componentLinearVelScale=<0.15 0.15 0.15>
    physicsBodyDefinitions [{
      sphereCollisionRadius=0
      boundBone {boneName=dyn_belt_tassel_m_parent}
      boxExtents=<12 12 4>
      constraintSetup {angularLimitsMax=<0 15 30>, angularLimitsMin=<0 0 -30>, angularTarget=<1 0 0>}
      localJointOffset=<3.5 0 0>
     }]
    relativeSpaceBone {boneName=root}
    sphericalLimits [{limitRadius=8.95, drivingBone {boneName=pelvis}, sphereLocalOffset=<-3 0 5>}]
   }
   n54 = Anim Dynamics - Bone: dyn_belt_tassel_n_parent(ComponentPose=@n53, Alpha=1, ExternalForce="0,0,0") {
    angularDampingOverride=0.9999
    angularSpringConstant=50
    bAngularSpring=true
    bGravityOverrideInSimSpace=true
    bOverrideAngularDamping=true
    bOverrideLinearDamping=true
    bUseSphericalLimits=true
    linearDampingOverride=0.999
    simulationSpace=RootRelative
    alphaBoolBlend {blendOption=Linear}
    boundBone {boneName=dyn_belt_tassel_n_parent}
    componentAppliedLinearAccClamp=<3000 3000 3000>
    componentLinearAccScale=<0.15 0.15 0.15>
    componentLinearVelScale=<0.15 0.15 0.15>
    physicsBodyDefinitions [{
      sphereCollisionRadius=0
      boundBone {boneName=dyn_belt_tassel_n_parent}
      boxExtents=<12 12 4>
      constraintSetup {angularLimitsMax=<0 5 30>, angularLimitsMin=<0 -25 -30>, angularTarget=<1 0 0>}
      localJointOffset=<3.5 0 0>
     }]
    relativeSpaceBone {boneName=root}
    sphericalLimits [{limitRadius=19.5, drivingBone {boneName=pelvis}, sphereLocalOffset=<-5 -10 0>}]
   }
   n55 = Anim Dynamics - Bone: dyn_belt_tassel_o_parent(ComponentPose=@n54, Alpha=1, ExternalForce="0,0,0") {
    angularDampingOverride=0.9999
    angularSpringConstant=50
    bGravityOverrideInSimSpace=true
    bOverrideAngularDamping=true
    bOverrideLinearDamping=true
    bUseSphericalLimits=true
    linearDampingOverride=0.9999
    simulationSpace=RootRelative
    alphaBoolBlend {blendOption=Linear}
    boundBone {boneName=dyn_belt_tassel_o_parent}
    componentAppliedLinearAccClamp=<3000 3000 3000>
    componentLinearAccScale=<0.15 0.15 0.15>
    componentLinearVelScale=<0.15 0.15 0.15>
    physicsBodyDefinitions [{
      sphereCollisionRadius=0
      boundBone {boneName=dyn_belt_tassel_o_parent}
      boxExtents=<12 12 4>
      constraintSetup {angularLimitsMax=<0 60 30>, angularLimitsMin=<0 20 -30>, angularTarget=<1 0 0>}
      localJointOffset=<3.5 0 0>
     }]
    relativeSpaceBone {boneName=root}
    sphericalLimits [{limitRadius=15, drivingBone {boneName=pelvis}, sphereLocalOffset=<-8 -5 0>}]
   }
   n56 = Anim Dynamics - Bone: dyn_belt_tassel_p_parent(ComponentPose=@n55, Alpha=1, ExternalForce="0,0,0") {
    angularDampingOverride=0.9999
    angularSpringConstant=50
    bAngularSpring=true
    bGravityOverrideInSimSpace=true
    bOverrideAngularDamping=true
    bOverrideLinearDamping=true
    bUseGravityOverride=true
    bUseSphericalLimits=true
    linearDampingOverride=0.9999
    simulationSpace=RootRelative
    alphaBoolBlend {blendOption=Linear}
    boundBone {boneName=dyn_belt_tassel_p_parent}
    componentAppliedLinearAccClamp=<3000 3000 3000>
    componentLinearAccScale=<0.15 0.15 0.15>
    componentLinearVelScale=<0.15 0.15 0.15>
    gravityOverride {z=-980}
    physicsBodyDefinitions [{
      sphereCollisionRadius=0
      boundBone {boneName=dyn_belt_tassel_p_parent}
      boxExtents=<12 12 4>
      constraintSetup {angularLimitsMax=<0 60 10>, angularLimitsMin=<0 30 -30>, angularTarget=<0 0 0>}
      localJointOffset=<3.5 0 0>
     }]
    relativeSpaceBone {boneName=root}
    sphericalLimits [
     {limitRadius=20, drivingBone {boneName=pelvis}, sphereLocalOffset=<-8 -9.5 0>}
     {limitRadius=10, drivingBone {boneName=thigh_l}, sphereLocalOffset=<-15 0 5>}
    ]
   }
   n57 = "Transform (Modify) Bone - Bone: attach_backpack"!(
    ComponentPose=@n56
    bAlphaBoolEnabled:bool=Get SilentTempoPack() -> (SilentTempoPack:bool)
    Translation="0,-4.5,0"
    Rotation="0,0,0"
    Scale="1,1,1"
   ) -> (Pose) {
    alphaInputType=Bool
    translationMode=BMM_Additive
    translationSpace=BCS_BoneSpace
    alphaBoolBlend {blendOption=Linear}
    boneToModify {boneName=attach_backpack}
    translation {y=-4.5}
   }
   n58 = Control Rig!(
    Source=Component To Local(ComponentPose=@n57)
    Alpha
    game_time_in_seconds:double=Get game_time_in_seconds() -> (game_time_in_seconds:double)
   ) -> (Pose) {
    alphaBoolBlend {blendOption=Linear}
    controlRigAssetReference {blueprintRigClass=&/F_MED_SilentTempo_Clock_CtrlRig.F_MED_SilentTempo_Clock_CtrlRig_C}
   }
   n59 = Layered blend per bone(BasePose=@n58, BlendPoses_0=Slot 'DefaultSlot'(), BlendWeights_0=Get Built In Emote() -> (Built In Emote:double)) {layerSetup [{branchFilters [{boneName=dfrm_clock_hour}, {boneName=dfrm_clock_minute}, {boneName=dfrm_clock_second}]}]}
   Output Pose(Result=@n59) {layerGroup=None}
  ]}
 EventGraph {nodes [
   Event OnCharacterCustomizationLoaded()
   > n1 = Branch(Condition=BackpackItemDefinitionHasAnyMetaTag!(Tags:Array of Gameplay Tag Structures=Make Array("[0]"=(TagName="Cosmetics.MetaTags.SilentTempo")), bRequiresEquipped:bool=false) -> (ReturnValue:bool))
   > Set SilentTempoPack(SilentTempoPack:bool=true) -> (Output_Get:bool)
   Set SilentTempoPack(execute=@n1.else, SilentTempoPack=false)
   Event BlueprintInitializeAnimation()
   > Set Skel_Mesh(Skel_Mesh=GetOwnerSkeletalMeshComponent())
   > n2 = Cast To PlayerPawn_Athena(Object=TryGetPawnOwner())
   > n3 = Set Pawn Owner(Pawn Owner:Player Pawn Athena Object Reference=@n2) -> (Output_Get:Player Pawn Athena Object Reference)
   > Set Body_Mesh(Body_Mesh=GetSkeletalMeshForPartType(self=@n3))
   Event BlueprintUpdateAnimation()
   > n4 = Sequence!(execute) -> (then_0, then_1, then_2:Exec, then_3:Exec, then_4:Exec, then_5:Exec)
   Set HeadRotate_Alpha(
    execute=@n4.then_1
    HeadRotate_Alpha:double=GetAlphaFromReferenceTransformDelta(
     SkeletalMeshComponent=Get Body_Mesh()
     BoneName=head
     BaseName=spine_05
     DeltaComponentType=QuaternionTwist
     InputRange=(X=-33,Y=0)
     TwistAxis="0,0,1"
    )
   ) -> (Output_Get:double)
   Set game_time_in_seconds(execute=@n4.then_0, game_time_in_seconds:double=GetGameTimeInSeconds()) -> (Output_Get:double)
   n5 = PlayerMovement(execute=@n4.then_5, AnimInstance=Self-Reference(), XDotProductCoefficient=1, YDotProductCoefficient=1, ZDotProductCoefficient=1)
   > Set left_right_movement_alpha(left_right_movement_alpha:double=@n5) -> (Output_Get:double)
   > Set forwards_backward_movement_alpha(forwards_backward_movement_alpha:double=@n5.YMovement) -> (Output_Get:double)
   n6 = Get Skel_Mesh()
   Set spine_twist_right_delta_ref_alpha(
    execute=@n4.then_3
    spine_twist_right_delta_ref_alpha:double=GetAlphaFromReferenceTransformDelta(
     SkeletalMeshComponent=@n6
     BoneName=spine_04
     BaseName=pelvis
     DeltaComponentType=QuaternionTwist
     InputRange=(X=-7,Y=-31)
     TwistAxis="1,0,0"
    )
   ) -> (Output_Get:double)
   > Set pelvis_tilt_left_alpha(pelvis_tilt_left_alpha:double=GetAlphaFromReferenceTransformDelta(SkeletalMeshComponent=@n6, BoneName=spine_02, BaseName=pelvis, DeltaComponentType=QuaternionTwist, InputRange=(X=-6,Y=6), TwistAxis="0,1,0")) -> (Output_Get:double)
   n7 = Gravity Override State Function DEPRECATED(execute=@n4.then_2, CharacterPartAnimInstance=Self-Reference(), GameTimeInSeconds=GetGameTimeInSeconds())
   > Set gravity_override(gravity_override=@n7)
   n8 = Branch(execute=@n4.then_4, Condition=Get bIsInFrontEnd())
   > Set Built In Emote(Built In Emote:double=GetCurveValue(self=GetOwnerFrontEndAnimInstance(), CurveName=emote_BuiltIn)) -> (Output_Get:double)
   Set Built In Emote(
    execute=@n8.else
    Built In Emote=GetCurveValue(
     self=GetOwnerFortPlayerAnimInstance!(self:Custom Character Part Anim Instance Object Reference) -> (ReturnValue:Fort Player Anim Instance Object Reference)
     CurveName=emote_BuiltIn
    )
   )
  ]}
]
variables [
 Body_Mesh {type=SkeletalMeshComponent}
 Built In Emote {type=number}
 forwards_backward_movement_alpha {type=number}
 game_time_in_seconds {type=number}
 gravity_override {type=Vector}
 HeadRotate_Alpha {type=number}
 left_right_movement_alpha {type=number}
 Pawn Owner {type=PlayerPawn_Athena_C}
 pelvis_tilt_left_alpha {type=number}
 SilentTempoPack {type=boolean}
 Skel_Mesh {type=SkeletalMeshComponent}
 spine_twist_right_delta_ref_alpha {type=number}
]
