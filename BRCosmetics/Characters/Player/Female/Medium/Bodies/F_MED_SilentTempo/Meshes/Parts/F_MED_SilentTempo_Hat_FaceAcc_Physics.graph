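# Physics asset for the SilentTempo female medium cosmetic's Hat_FaceAcc mesh part, defining physics bodies for its dynamic hair bones.
# Five hair groups (a_r, e_l, f_r, g_l, h_l) use capsules constrained back to small base spheres in chains or branches, with tight twist and modest swing limits.
# Neck and upper-spine capsules have no constraints and provide body volumes alongside the hair.
# One fully free angular constraint links the g_l tip to f_r_01 across groups.
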PhysicsAsset /BRCosmetics/Characters/Player/Female/Medium/Bodies/F_MED_SilentTempo/Meshes/Parts/F_MED_SilentTempo_Hat_FaceAcc_Physics
bodies {
 dyn_hair_a_r_base [{shapeType=Sphere, radius=0.5, shapeName=dyn_hair_a_r_base_sphere}]
 dyn_hair_a_r_parent [{shapeType=Capsule, length=2.6314, radius=1.1819, shapeName=dyn_hair_a_r_parent_capsule, center=<2.0759 0 0>, rotation=<rot 90 0 0>}]
 dyn_hair_e_l_01 [{shapeType=Capsule, length=0.5954, radius=0.8696, shapeName=dyn_hair_e_l_01_capsule, center=<0.5267 0 0>, rotation=<rot 90 0 0>}]
 dyn_hair_e_l_02 [{shapeType=Capsule, length=0.5954, radius=0.8696, shapeName=dyn_hair_e_l_02_capsule, center=<0.5267 0 0>, rotation=<rot 90 0 0>}]
 dyn_hair_e_l_base [{shapeType=Sphere, radius=0.5, shapeName=dyn_hair_e_l_base_sphere}]
 dyn_hair_e_l_parent [{shapeType=Capsule, length=1.5536, radius=1.3594, shapeName=dyn_hair_e_l_parent_capsule, center=<1.6108 0 0>, rotation=<rot 90 0 0>}]
 dyn_hair_f_r_01 [{shapeType=Capsule, length=2.9644, radius=1.5425, shapeName=dyn_hair_f_r_01_capsule, center=<1.6188 0 0>, rotation=<rot 90 0 0>}]
 dyn_hair_f_r_02 [{shapeType=Capsule, length=4.5889, radius=1.5063, shapeName=dyn_hair_f_r_02_capsule, center=<2.8842 0 0.2685>, rotation=<rot 76.3477 180 180>}]
 dyn_hair_f_r_base [{shapeType=Sphere, radius=0.5, shapeName=dyn_hair_f_r_base_sphere}]
 dyn_hair_f_r_parent [{shapeType=Capsule, length=2.2979, radius=1.3246, shapeName=dyn_hair_f_r_parent_capsule, center=<1.5866 0 0>, rotation=<rot 90 0 0>}]
 dyn_hair_g_l_01 [{shapeType=Capsule, length=7.5039, radius=3.1595, shapeName=dyn_hair_g_l_01_capsule, center=<2.7384 0 0>, rotation=<rot 0 0 90>}]
 dyn_hair_g_l_02 [{shapeType=Capsule, length=9.0227, radius=3.8485, shapeName=dyn_hair_g_l_02_capsule, center=<2.7828 0 -0.4603>, rotation=<rot 0 0 90>}]
 dyn_hair_g_l_base [{shapeType=Sphere, radius=0.5, shapeName=dyn_hair_g_l_base_sphere}]
 dyn_hair_g_l_parent [
  {shapeType=Capsule, length=2.1139, radius=1.6713, shapeName=dyn_hair_g_l_parent_capsule, center=<1.9149 0 0>, rotation=<rot 90 0 0>}
  {shapeType=Capsule, length=7.2569, radius=2.6137, shapeName=dyn_hair_g_l_parent_capsule, center=<3.0878 0 0>, rotation=<rot 0 0 90>}
 ]
 dyn_hair_h_l_01 [{shapeType=Capsule, length=4.1472, radius=2.1899, shapeName=dyn_hair_h_l_01_capsule, center=<2.7774 0 0>, rotation=<rot 90 0 0>}]
 dyn_hair_h_l_02 [{shapeType=Capsule, length=1.1013, radius=1.265, shapeName=dyn_hair_h_l_02_capsule, center=<1.5176 0 0>, rotation=<rot 90 0 0>}]
 dyn_hair_h_l_03 [{shapeType=Capsule, length=3.4448, radius=1.5774, shapeName=dyn_hair_h_l_03_capsule, center=<2.1988 0 0>, rotation=<rot 90 0 0>}]
 dyn_hair_h_l_base [{shapeType=Sphere, radius=0.5, shapeName=dyn_hair_h_l_base_sphere}]
 dyn_hair_h_l_parent [{shapeType=Capsule, length=2.2924, radius=1.8056, shapeName=dyn_hair_h_l_parent_capsule, center=<2.4147 0 0>, rotation=<rot 90 0 0>}]
 dyn_hair_offset [{shapeType=Sphere, radius=0.5, shapeName=dyn_hair_offset_sphere}]
 neck_02 [{shapeType=Capsule, length=10.8993, radius=7.3728, shapeName=neck_02_capsule, center=<0 1.4736 0>, rotation=<rot 90 0 0>}]
 spine_05 [{shapeType=Capsule, length=21.4882, radius=8.103, shapeName=spine_05_capsule, center=<2.4015 0 0>}]
}
constraints [
 {bone1Name=dyn_hair_h_l_03, bone2Name=dyn_hair_h_l_02, swing1LimitDegrees=10, swing2LimitDegrees=10, twistLimitDegrees=3}
 {bone1Name=dyn_hair_h_l_02, bone2Name=dyn_hair_h_l_parent, swing1LimitDegrees=10, swing2LimitDegrees=10, twistLimitDegrees=3}
 {bone1Name=dyn_hair_h_l_parent, bone2Name=dyn_hair_h_l_base, swing1LimitDegrees=10, swing2LimitDegrees=10, twistLimitDegrees=3}
 {bone1Name=dyn_hair_h_l_01, bone2Name=dyn_hair_h_l_parent, swing1LimitDegrees=10, swing2LimitDegrees=10, twistLimitDegrees=3}
 {bone1Name=dyn_hair_g_l_02, bone2Name=dyn_hair_g_l_01, swing1LimitDegrees=20, swing2LimitDegrees=20, twistLimitDegrees=3}
 {bone1Name=dyn_hair_g_l_01, bone2Name=dyn_hair_g_l_parent, swing1LimitDegrees=20, swing2LimitDegrees=20, twistLimitDegrees=3}
 {bone1Name=dyn_hair_g_l_parent, bone2Name=dyn_hair_g_l_base, swing1LimitDegrees=20, swing2LimitDegrees=20, twistLimitDegrees=3}
 {bone1Name=dyn_hair_f_r_02, bone2Name=dyn_hair_f_r_parent, swing1LimitDegrees=20, swing2LimitDegrees=20, twistLimitDegrees=3}
 {bone1Name=dyn_hair_f_r_01, bone2Name=dyn_hair_f_r_parent, swing1LimitDegrees=20, swing2LimitDegrees=20, twistLimitDegrees=3}
 {bone1Name=dyn_hair_f_r_parent, bone2Name=dyn_hair_f_r_base, swing1LimitDegrees=20, swing2LimitDegrees=20, twistLimitDegrees=3}
 {bone1Name=dyn_hair_e_l_02, bone2Name=dyn_hair_e_l_parent, swing1LimitDegrees=10, swing2LimitDegrees=10, twistLimitDegrees=3}
 {bone1Name=dyn_hair_e_l_01, bone2Name=dyn_hair_e_l_parent, swing1LimitDegrees=10, swing2LimitDegrees=10, twistLimitDegrees=3}
 {bone1Name=dyn_hair_e_l_parent, bone2Name=dyn_hair_e_l_base, swing1LimitDegrees=10, swing2LimitDegrees=10, twistLimitDegrees=3}
 {bone1Name=dyn_hair_a_r_parent, bone2Name=dyn_hair_a_r_base, swing1LimitDegrees=20, swing2LimitDegrees=20, twistLimitDegrees=3}
 {
  bone1Name=dyn_hair_g_l_02
  bone2Name=dyn_hair_f_r_01
  swing1LimitDegrees=45
  swing1Motion=Free
  swing2LimitDegrees=45
  swing2Motion=Free
  twistLimitDegrees=45
  twistMotion=Free
 }
]
